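# Translucent unlit Niagara ribbon material for the FilmSparkle aura cosmetic, drawing a tiled film strip texture tinted by particle color.
# A panning sine wave through a gradient ramp adds shine, while texture channels supply color and opacity masks.
# Opacity also uses end fades along U and depth fade, bypassed on mobile.
# A dynamic parameter controls wave count, offset and a directional noise burnaway with a glowing hot edge.
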Material /SparksCosmetics/Items/Auras/Assets/NS_Aura_FilmSparkle/M_Aura_FilmSparkle_FilmStrip_Ribbons
graph {
 properties {
  bAutomaticallySetUsageInEditor=false
  bUsedWithNiagaraRibbons=true
  blendMode=BLEND_Translucent
  shadingModel=MSM_Unlit
  usedShadingModels=Unlit
 }
 nodes [
  n1 = NamedRerouteUsage()
  n2 = ParticleColor()
  n3 = Multiply(A=NamedRerouteUsage(), B=@n2.RGB)
  n4 = VectorParameter() {group=Color, parameterName=Hot Edge Color, defaultValue=<rgba 15 5.6144 1.3863 1>}
  n5 = Multiply(A=@n4.RGB, B=@n4.RGB)
  n6 = NamedRerouteUsage()
  MaterialFunctionCall(Emissive=Add(A=Multiply(A=@n1, B=@n3), B=Multiply(A=@n5, B=@n6))) {materialFunction=&MF_FX_EmissiveAdjustments}
  n7 = Multiply(A=@n2.A, B=NamedRerouteUsage())
  n8 = NamedRerouteUsage()
  n9 = DepthFade(FadeDistance=ScalarParameter() {defaultValue=20, group=Opacity, parameterName=FadeDistance})
  Multiply(
   A=Multiply(A=Multiply(A=@n7, B=@n8), B=NamedRerouteUsage())
   B=ShadingPathSwitch(Default=@n9, Deferred=@n9, Mobile=Constant() {r=1}) {desc=Migrated from FeatureLevelSwitch}
  )
  n10 = TextureCoordinate()
  n11 = Saturate(None=Multiply(A=ComponentMask(None=@n10) {r=true}, B=ScalarParameter() {defaultValue=2, group=Opacity, parameterName=Front End Fade Strength}) {constB=8})
  n12 = ComponentMask(None=TextureCoordinate()) {r=true}
  n13 = Saturate(None=Multiply(A=OneMinus(None=@n12), B=ScalarParameter() {defaultValue=5, group=Opacity, parameterName=Back End Fade Strength}))
  NamedRerouteDeclaration(None=Multiply(A=Multiply(A=@n11, B=@n11), B=Multiply(A=@n13, B=@n13))) {name=UV Mask, nodeColor=<rgba 1 0.6328 0 1>}
  n14 = ScalarParameter() {defaultValue=8, parameterName=Tile u pattern}
  n15 = ScalarParameter() {defaultValue=1, parameterName=Tile v pattern}
  n16 = TextureSampleParameter2D(UVs=Multiply(A=TextureCoordinate(), B=AppendVector(A=@n14, B=@n15))) {parameterName=Mask Texture, texture=&/T_Aura_FilmSparkle_FilmStrip}
  n17 = StaticComponentMaskParameter(None=@n16.RGBA) {defaultR=true, parameterName=OpacityMask Channel}
  n18 = ScalarParameter() {defaultValue=2, group=Opacity, parameterName=Opacity Mask Strength}
  NamedRerouteDeclaration(None=Saturate(None=Multiply(A=@n17, B=@n18))) {name=Texture Opacity Mask, nodeColor=<rgba 1 0.6797 0 1>}
  MaterialExpressionTextureCoordinate_5 = TextureCoordinate() {coordinateIndex=1}
  n19 = Convert(None=@MaterialExpressionTextureCoordinate_5) {
   nodeName=Break Float2
   convertInputs [{
     type=Vector2
     expressionInput {
      expression=&self:MaterialExpressionTextureCoordinate_5
      inputName=None
      mask=0
      maskA=0
      maskB=0
      maskG=0
      maskR=0
      outputIndex=0
     }
    }]
   convertOutputs [{type=Scalar}, {type=Scalar}]
  }
  n20 = NamedRerouteUsage()
  n21 = Multiply(
   A=Multiply(A=NamedRerouteUsage(), B=ScalarParameter() {defaultValue=2.5, group=Color, parameterName=Wave Count})
   B=OneMinus(None=Add(A=@n19, B=@n20))
  ) {constB=3}
  n22 = Multiply(A=Time(), B=ScalarParameter() {defaultValue=-0.25, group=Color, parameterName=Wave Pan Speed})
  n23 = MaterialFunctionCall(
   Input=Sine(None=Add(A=@n21, B=@n22))
   Input Low=ScalarParameter() {defaultValue=-1, group=Color, parameterName=Sine Gradient Low}
   Input High=ScalarParameter() {defaultValue=1, group=Color, parameterName=Sine Gradient High}
  ) {materialFunction=&RemapValue_0-1}
  NamedRerouteDeclaration(None=TextureSample(
    UVs=Saturate(None=@n23.Result)
    Tex=TextureObjectParameter() {group=Color, parameterName=Gradient Texture, texture=/Game/Effects/Materials/SwingTrails/Textures/T_FX_ColorRamp_Linear}
   ).RGB) {name=Sine Wave Shine Gradient, nodeColor=<rgba 0.4297 0 1 1>}
  n24 = StaticComponentMaskParameter(None=@n16.RGBA) {defaultG=true, parameterName=Color Mask Channel}
  n25 = LinearInterpolate(A=@n24, B=OneMinus(None=@n24), Alpha=ScalarParameter() {defaultValue=1, group=Color, parameterName=Invert Color Mask Lerp})
  n26 = ScalarParameter() {defaultValue=0.35, group=Color, parameterName=Color Mask add}
  NamedRerouteDeclaration(None=Saturate(None=Add(A=@n25, B=@n26))) {name=Texture Color Mask, nodeColor=<rgba 1 0 0.1732 1>}
  n27 = DynamicParameter() {defaultValue=<rgba 1 1 1 0.2>, paramNames [wave count mod, Offset, Burnaway, Burnaway Thickness]}
  NamedRerouteDeclaration(None=@n27.wave count mod) {name=Dynamic 0 - wave count mod, nodeColor=<rgba 1 0 0.0147 1>}
  n28 = Constant() {r=1}
  n29 = Add(A=OneMinus(None=NamedRerouteUsage())) {constB=0.015}
  n30 = NamedRerouteUsage()
  n31 = Add(A=TextureCoordinate() {coordinateIndex=1}, B=AppendVector(A=@n30, B=@n30))
  n32 = AppendVector(
   A=ScalarParameter() {defaultValue=2, group=Burnaway, parameterName=Tile U Noise}
   B=ScalarParameter() {defaultValue=0.7, group=Burnaway, parameterName=Tile V Noise}
  )
  n33 = StaticComponentMaskParameter(None=TextureSample(UVs=Multiply(A=@n31, B=@n32)) {texture=&T_Smooth_tile}.RGBA) {defaultR=true, group=Burnaway, parameterName=Burnaway Noise Texture Channel}
  n34 = ComponentMask(None=TextureCoordinate()) {r=true}
  n35 = Saturate(None=Multiply(A=@n34, B=ScalarParameter() {defaultValue=0.4, group=Burnaway, parameterName=burnaway directional fade}) {constB=8})
  n36 = LinearInterpolate(Alpha=Saturate(None=Add(A=@n33, B=@n35))) {constA=0.1}
  NamedRerouteDeclaration(None=If(A=NamedRerouteUsage(), B=@n28, "A > B"=@n28, "A == B"=@n28, "A < B"=SmoothStep(Min=@n29, Max=@n29, Value=@n36))) {name=BurnawayOpacity, nodeColor=<rgba 0.1016 1 0 1>}
  n37 = Constant()
  n38 = NamedRerouteUsage()
  n39 = SmoothStep(Min=@n29, Max=Add(A=@n29, B=Max(A=@n38) {constB=0.001}), Value=@n36)
  NamedRerouteDeclaration(None=If(A=NamedRerouteUsage(), B=@n28, "A > B"=@n37, "A == B"=@n37, "A < B"=OneMinus(None=@n39))) {name=BurnawayEmissive, nodeColor=<rgba 1 0 0.4219 1>}
  NamedRerouteDeclaration(None=@n27.Offset) {name=Dynamic 0 - offset, nodeColor=<rgba 0.078 1 0 1>}
  NamedRerouteDeclaration(None=@n27.Burnaway) {name=Dynamic 0 - Burnaway, nodeColor=<rgba 0 0.029 1 1>}
  NamedRerouteDeclaration(None=@n27.Burnaway Thickness) {name=Dynamic 0 - Burnaway Thickness, nodeColor=<rgba 0.7177 0.8214 1 1>}
  OneMinus(None=@n34)
 ]
}
