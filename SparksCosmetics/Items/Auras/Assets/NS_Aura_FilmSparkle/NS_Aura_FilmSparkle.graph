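# Niagara system for the FilmSparkle aura cosmetic, adding warm golden glows, rays, flickering motes and rising mesh ribbons to a character.
# Emitters pin particles to skeletal mesh bones, using spawn groups, partitions and switch inputs to vary each layer within one emitter.
# Waveforms and curl noise animate the motes, while large-body, front-end preview and time-of-day inputs adjust size, speed or brightness.
# The standalone Motes_Flares002 emitter is disabled.
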NiagaraSystem /SparksCosmetics/Items/Auras/Assets/NS_Aura_FilmSparkle/NS_Aura_FilmSparkle
dependencies {
 usedDataInterfaces [
  &NiagaraDataInterfaceArrayFloat3
  &NiagaraDataInterfaceCurve
  &NiagaraDataInterfaceColorCurve
  &NiagaraDataInterfaceArrayInt32
  &NiagaraDataInterfaceArrayFloat
  &NiagaraDataInterfaceCamera
  &NiagaraDataInterfaceParticleRead
  &NiagaraDataInterfaceSkeletalMesh
 ]
 usedDynamicInputs []
 usedModules [
  &self:SystemSpawnScript.SystemScriptSource.SystemScriptGraph.NiagaraNodeAssignment_0.SetVariables_#1
  &SystemState
  &self:SystemSpawnScript.SystemScriptSource.SystemScriptGraph.NiagaraNodeAssignment_1.SetVariables_#2
  &self:SystemSpawnScript.SystemScriptSource.SystemScriptGraph.NiagaraNodeAssignment_2.SetVariables_#3
 ]
}
emitters {
 Decal_GlowRays {
  emitterUpdateScript [
   &EmitterState()
   &SpawnBurst_Instantaneous(Spawn Count=5, Use Spawn Probability=false, UseLoopCountLimit=true)
   &SpawnBurst_Instantaneous(Spawn Count=3, Spawn Group=1, Use Spawn Probability=false, UseLoopCountLimit=true) {module=SpawnBurst_Instantaneous001}
  ]
  particleSpawnScript [
   /SetVariables(Particles.VisibilityTag=&SpawnGroup()) {isSetParameters=true}
   &PartitionParticles(Number Of Partitions=&MergeHelper_IntSwitch(0=5, 1=3, 2=?, 3=?, 4=?, 5=?, Index=&SpawnGroup(), Num=::NewEnumerator0))
   &InitializeParticle(
    Color=&MergeHelper_ColorSwitch(
     0=&Multiply_LinearColorByFloat(Linear Color=<rgba 1.75 0.9825 0.1848 1>, Float=1.5)
     1=&MultiplyLinearColor(
      Linear Color=&MergeHelper_ColorSwitch(
       0=<rgba 4 2.1406 0.6484 1>
       1=<rgba 2.5 1.3379 0.4053 1>
       2=<rgba 2.5 1.3379 0.4053 1>
       3=?
       Index=$Particles.PartitionParticles.Partition
       Num=::NewEnumerator1
      )
      Scale Factor=&MakeVector4_FromFloat(Float=&MakeCustomFloatFromBool(Bool=$User.bIsFrontEndPreview, True Float=0.7, False Float=1))
     )
     2=?
     3=?
     Index=&SpawnGroup()
     Num=::NewEnumerator0
    )
    Lifetime=&MergeHelper_FloatSwitch(0=3, 1=0.3, 2=?, 3=?, Index=&SpawnGroup(), Num=::NewEnumerator0)
    Sprite Rotation Mode=::NewEnumerator0
    Sprite Size Mode=::NewEnumerator3
    Uniform Sprite Size=&MergeHelper_FloatSwitch(
     0=&Multiply_Float(
      A=&SelectFloatFromArray(
       Direct Array Index=&SpawnGroup()
       Float Selection Array=DataInterface() {propertyValues {FloatData [40,40,20,20,30]}}
      )
      B=&MakeCustomFloatFromBool(Bool=$User.IsLargeBody, True Float=1.5, False Float=1)
     )
     1=&Multiply_Float(
      A=&SelectFloatFromArray(Direct Array Index=$Particles.PartitionParticles.Partition, Float Selection Array=DataInterface() {propertyValues {FloatData [125,100,100]}})
      B=&MakeCustomFloatFromBool(Bool=$User.IsLargeBody, True Float=1.1, False Float=1)
     )
     2=?
     3=?
     Index=&SpawnGroup()
     Num=::NewEnumerator0
    )
    UsePositionOffset=false
    Write Material Random=true
   ) {module=InitializeParticle001}
   &CameraOffset(Camera Offset Amount=&MergeHelper_FloatSwitch(0=5, 1=-15, 2=?, 3=?, Index=&SpawnGroup(), Num=::NewEnumerator0))
  ]
  particleUpdateScript [
   &ParticleState(
    Kill Particles When Lifetime Has Elapsed=false
    Let Infinitely Lived Particles Die When Emitter Deactivates=true
    Loop Particles Lifetime=true
   )
   &SkeletalMeshLocation(
    Bone / Socket Index=&MergeHelper_IntSwitch(
     0=&SelectIntFromArray(Direct Array Index=$Particles.PartitionParticles.Partition, Int Selection Array=DataInterface() {propertyValues {IntData [0,1,2,3,4]}})
     1=&SelectIntFromArray(Direct Array Index=$Particles.PartitionParticles.Partition, Int Selection Array=DataInterface() {propertyValues {IntData [7,1,2]}})
     2=?
     3=?
     4=?
     5=?
     Index=&SpawnGroup()
     Num=::NewEnumerator0
    )
    Bone Sampling Mode=::NewEnumerator4
    MaskSpawnGroup=false
    Orientation Sampling=::NewEnumerator1
    Sampled Position Offset=&MergeHelper_VectorSwitch(
     0=<0 0 0>
     1=&Multiply_Vector(
      A=&Multiply_VectorByFloat(
       Vector=&SelectVectorFromArray(
        Direct Array Index=$Particles.PartitionParticles.Partition
        Vector Selection Array=DataInterface() {propertyValues {FloatData [<7 0 0>, <5 0 0>, <-5 0 0>]}}
       )
       Float=1
      )
      B=&MakeCustomVectorFromBool(Bool=$User.IsLargeBody, True Vector=<1 1 1.25>, False Vector=<1 1 1>)
     )
     2=?
     3=?
     Index=&SpawnGroup()
     Num=::NewEnumerator0
    )
    Skeletal Mesh=$System.Skeletal Mesh
    Use Sampled Mesh Attribute as Ribbon ID=false
   ) {module=SkeletalMeshLocation003}
   &DynamicMaterialParameters(
    Index 0 Param 1=&SelectFloatFromArray(Direct Array Index=$Particles.PartitionParticles.Partition, Float Selection Array=DataInterface() {propertyValues {FloatData [5,5,5,5,3]}})
    Index 0 Param 3=1
    Module.Param0WriteEnabled=true
    Module.Param1WriteEnabled=true
    Module.Param2WriteEnabled=true
    Module.Param3WriteEnabled=true
   )
   &SpriteRotationRate(Rotation Rate=&MergeHelper_FloatSwitch(
     0=0
     1=&MakeCustomFloatFromBool(
      Bool=&SetBoolByIntComparison(A=$Particles.PartitionParticles.Partition, B=0, Comparison Type=::NewEnumerator0)
      True Float=&RandomRangeFloat(Minimum=10, Maximum=15)
      False Float=&RandomRangeFloat(Minimum=-10, Maximum=-15)
     )
     2=?
     3=?
     Index=&SpawnGroup()
     Num=::NewEnumerator0
    ))
   &ScaleColor(
    Scale Alpha=&MergeHelper_FloatSwitch(
     0=1
     1=&FloatFromCurve(
      FloatCurve=DataInterface() {propertyValues {Curve {keys [
          <at -0.0009 0.85 RCIM_Cubic>
          {arriveTangent=0.4254, interpMode=RCIM_Cubic, leaveTangent=0.4254, time=0.1742, value=0.95}
          {arriveTangentWeight=0.0596, interpMode=RCIM_Cubic, leaveTangentWeight=0.0823, tangentMode=RCTM_User, time=0.3518, value=1}
          {arriveTangentWeight=0.0838, interpMode=RCIM_Cubic, leaveTangentWeight=0.1397, tangentMode=RCTM_User, time=0.5968, value=1}
          <at 1.0052 0.85 RCIM_Cubic>
         ]}}}
      CurveIndex=Linked() {linkedVariable {name=NPC.FortniteNPC.TimeOfDay_0-1, type=NiagaraFloat}}
     )
     2=?
     3=?
     Index=&SpawnGroup()
     Num=::NewEnumerator0
    )
    ScaleA=true
    ScaleRGB=true
   )
  ]
  rendererClasses [&NiagaraSpriteRendererProperties]
  renderers [
   {rendererClass=&NiagaraSpriteRendererProperties}
   {rendererClass=&NiagaraSpriteRendererProperties, rendererIndex=1}
  ]
 }
 Motes_Flares002 {
  bEnabled=false
  emitterUpdateScript [
   &EmitterState()
   &SpawnRate(Spawn Probability=0.9, SpawnRate=16, Use Spawn Probability=true)
   &SpawnRate(Spawn Probability=0.8, SpawnGroup=1, SpawnRate=7, Use Spawn Probability=true) {module=SpawnRate001}
  ]
  particleSpawnScript [
   &PartitionParticles(Number Of Partitions=3, Partition Mode:ENiagara_PartitionMode_v2=::NewEnumerator5)
   /SetVariables(Particles.VisibilityTag=&SpawnGroup()) {isSetParameters=true, module=SetVariables_#4}
   &ShapeLocation(
    MaskSpawnGroup=false
    Non Uniform Scale=<1.2 1 1>
    Override Seed=false
    Rotation Coordinate Space=::Simulation
    Shape Origin=&VectorFromFloat(Value=0)
    Sphere Radius=15
    Sphere Surface Distribution=1
    UseOwnerScale=true
   )
   &InitializeParticle(
    Color=&RandomRangeLinearColor(Minimum=<rgba 1 0.4865 0.2632 1>, Maximum=<rgba 1 0.8182 0.0938 1>)
    Lifetime=&MergeHelper_FloatSwitch(
     0=&RandomRangeFloat(Minimum=1.2, Maximum=1.5)
     1=&RandomRangeFloat(Minimum=0.8, Maximum=1.5)
     2=?
     3=?
     Index=&SpawnGroup()
     Num=::NewEnumerator0
    )
    Sprite Rotation Mode=::NewEnumerator0
    Sprite Size Mode=::NewEnumerator3
    Uniform Sprite Size=&MergeHelper_FloatSwitch(
     0=&RandomRangeFloat(Minimum=1, Maximum=2)
     1=&RandomRangeFloat(Minimum=8, Maximum=15)
     2=?
     3=?
     Index=&SpawnGroup()
     Num=::NewEnumerator0
    )
    UsePositionOffset=false
    Write Material Random=true
   )
   &AddVelocity(
    Cone Angle=10
    Cone Axis=<0 0 1>
    Module.UseConeFalloff=false
    Module.UseVelDistribution=false
    Override Seed=false
    Rotation Coordinate Space=::World
    Velocity Mode=::NewEnumerator2
    Velocity Speed=&Multiply_Float(
     A=&MergeHelper_FloatSwitch(
      0=&RandomRangeFloat(Minimum=45, Maximum=60)
      1=&RandomRangeFloat(Minimum=15, Maximum=20)
      2=?
      3=?
      Index=&SpawnGroup()
      Num=::NewEnumerator0
     )
     B=&MakeCustomFloatFromBool(Bool=$User.IsLargeBody, True Float=1.3, False Float=1)
    )
   )
   &CameraOffset(Camera Offset Amount=-25)
  ]
  particleUpdateScript [
   /SetVariables(Particles.FlickerWaveform:NiagaraFloat=&Waveform(
     Global Phase=&Add_Float(A=$Engine.Time, B=&RandomRangeFloat(Minimum=0, Maximum=1))
     "Waveform [1]"=::NewEnumerator10
     "[1] Frequency"=15
     "[1] Amplitude Min/Max"=<0.2 1.2>
    )) {isSetParameters=true, module=SetVariables_#5}
   &SkeletalMeshLocation(
    Bone / Socket Index=7
    Bone Sampling Mode=::NewEnumerator4
    "Inherit Velocity (Scale)"=-0.07
    Inherited Velocity Speed Limit=25
    MaskSpawnGroup=false
    Orientation Sampling=::NewEnumerator1
    Position Sampling=::NewEnumerator1
    Sampled Position Offset=&AddVectorToPosition(
     Position=Linked() {linkedVariable {name=Particles.Initial.ShapeLocation.ShapePosition, type=Vector3f}}
     Vector=&MergeHelper_VectorSwitch(
      0=&SelectVectorFromArray(
       Array Sampling Mode=::NewEnumerator0
       Direct Array Index=?
       Vector Selection Array=DataInterface() {propertyValues {FloatData [<10 0 0>, <-5 0 -14>, <-5 0 14>]}}
       Random Seed=0
      )
      1=&SelectVectorFromArray(
       Array Sampling Mode=::NewEnumerator0
       Direct Array Index=?
       Vector Selection Array=DataInterface() {propertyValues {FloatData [<20 0 0>, <10 0 -12>, <10 0 12>]}}
       Random Seed=0
      )
      2=?
      3=?
      Index=&SpawnGroup()
      Num=::NewEnumerator0
     )
    )
    Skeletal Mesh=$System.Skeletal Mesh
    Velocity Sampling=::NewEnumerator0
    VelocitySpeedLimit=true
   )
   &ParticleState()
   &ScaleSpriteSize(Scale Factor=&MergeHelper_Vector2Switch(
     0=&Vector2DFromFloat(Value=&FloatFromCurve(
       FloatCurve=DataInterface() {propertyValues {Curve {keys [
           {arriveTangent=7.4528, interpMode=RCIM_Cubic, leaveTangent=7.4528, tangentMode=RCTM_User, time=0, value=0}
           <at 0.2 1.2 RCIM_Cubic>
           {arriveTangent=-2.7132, interpMode=RCIM_Cubic, leaveTangent=-2.7132, tangentMode=RCTM_User, time=1, value=0}
          ]}}}
       Scale Curve=&Multiply_Float(A=&FN_FrontEndFloatModifier(Gameplay=0.6, Front End=0.5), B=$Particles.FlickerWaveform)
      ))
     1=&Multiply_Vector2D_ByFloat(
      Float=&Multiply_Float(
       A=&FloatFromCurve(FloatCurve=DataInterface() {propertyValues {Curve {keys [
            {arriveTangent=7.4528, interpMode=RCIM_Cubic, leaveTangent=7.4528, tangentMode=RCTM_User, time=0, value=0}
            <at 0.2 1.2 RCIM_Cubic>
            {arriveTangent=-2.7132, interpMode=RCIM_Cubic, leaveTangent=-2.7132, tangentMode=RCTM_User, time=1, value=0}
           ]}}})
       B=/Game/Effects/Niagara/DynamicInputs/Math/FN_Fit(
        Clamp:NiagaraBool=false
        Value:NiagaraFloat=$Particles.FlickerWaveform
        SourceMin:NiagaraFloat=0.2
        SourceMax:NiagaraFloat=1.2
        DestMin:NiagaraFloat=0.5
        DestMax:NiagaraFloat=1
       )
      )
      Vector2D=&MakeVector2D(
       X=&FloatFromCurve(FloatCurve=DataInterface() {propertyValues {Curve {keys [<at 0 0.8>,<at 1 1>]}}})
       Y=&FloatFromCurve(FloatCurve=DataInterface() {propertyValues {Curve {keys [<at 0 1.2529>,<at 1 1>]}}})
      )
     )
     2=?
     3=?
     4=?
     Index=&SpawnGroup()
     Num=::NewEnumerator0
    ))
   &CurlNoiseForce(
    Bias Noise=true
    Bias Noise Field:Vector3f=<0 0 0.25>
    Noise Strength=&FloatFromCurve(
     FloatCurve=DataInterface() {propertyValues {Curve {
        defaultValue=1
        keys [{interpMode=RCIM_Cubic, tangentMode=RCTM_User, time=0, value=0}, {interpMode=RCIM_Cubic, tangentMode=RCTM_User, time=1, value=1}]
       }}}
     Scale Curve=30
    )
    Pan Noise=true
    Pan Noise Field=<0 0 1>
   )
   &ScaleColor(
    Scale RGB=&Multiply_VectorByFloat(
     Vector=&MakeVectorFromLinearColorRGB(Color=&ColorFromCurve(
       ColorCurve=DataInterface() {propertyValues {
         AlphaCurve {keys [<at 0 1 RCIM_Cubic>,<at 1 0 RCIM_Cubic>]}
         BlueCurve {keys [<at 0 1 RCIM_Cubic>,<at 1 1 RCIM_Cubic>]}
         GreenCurve {keys [<at 0 1 RCIM_Cubic>,<at 1 1 RCIM_Cubic>]}
         RedCurve {keys [<at 0 1 RCIM_Cubic>,<at 1 1 RCIM_Cubic>]}
        }}
       CurveIndex=&Frac_Float(Float=&Add_Float(A=&RandomRangeFloat(Minimum=0, Maximum=1), B=$Particles.NormalizedAge))
      ))
     Float=&FloatFromCurve(
      FloatCurve=DataInterface() {propertyValues {Curve {
         defaultValue=1
         keys [
          {interpMode=RCIM_Cubic, tangentMode=RCTM_User, time=0, value=1}
          {arriveTangent=-2.2239, interpMode=RCIM_Cubic, leaveTangent=-2.2239, tangentMode=RCTM_User, time=1, value=0.2}
         ]
        }}}
      Scale Curve=&Multiply_Float(
       A=4
       B=&FN_Fit(
        Clamp:NiagaraBool=false
        Value:NiagaraFloat=$Particles.FlickerWaveform
        SourceMin:NiagaraFloat=0.2
        SourceMax:NiagaraFloat=1.2
        DestMin:NiagaraFloat=0.75
        DestMax:NiagaraFloat=1.2
       )
      )
     )
    )
    ScaleA=true
    ScaleRGB=true
   )
   /SetVariables(Particles.Glow03_Size=&MergeHelper_Vector2Switch(
     0=&Multiply_Vector2D_ByFloat(
      Float=&FloatFromCurve(
       FloatCurve=DataInterface() {propertyValues {Curve {keys [
           <at 0 0 RCIM_Cubic>
           <at 0.2 1.3 RCIM_Cubic>
           {arriveTangent=-0.1423, interpMode=RCIM_Cubic, leaveTangent=-0.1423, tangentMode=RCTM_User, time=1, value=0}
          ]}}}
       Scale Curve=&Multiply_Float(A=&MakeFloatFromVector2d(Vector 2D=$Particles.SpriteSize), B=3.5)
      )
      Vector2D=<1 1.5>
     )
     1=&Multiply_Vector2D_ByFloat(
      Float=&FloatFromCurve(
       FloatCurve=DataInterface() {propertyValues {Curve {keys [
           <at 0 0 RCIM_Cubic>
           <at 0.2 1.3 RCIM_Cubic>
           {arriveTangent=-0.1423, interpMode=RCIM_Cubic, leaveTangent=-0.1423, tangentMode=RCTM_User, time=1, value=0}
          ]}}}
       Scale Curve=&Multiply_Float(A=&MakeFloatFromVector2d(Vector 2D=$Particles.SpriteSize), B=2)
      )
      Vector2D=<1 1>
     )
     2=?
     3=?
     4=?
     Index=&SpawnGroup()
     Num=::NewEnumerator0
    )) {isSetParameters=true, module=SetVariables_#6}
   /SetVariables(Particles.Glow03_Color=&MergeHelper_ColorSwitch(
     0=&MakeLinearColorFromVectorAndFloat(
      "Vector (RGB)"=&MakeVectorFromLinearColorRGB(Color=&MultiplyLinearColor(Linear Color=$Particles.Color, Scale Factor=&MakeVector4_FromLinearColor(LinearColor=$Particles.Color)))
      "Float (Alpha)"=&Multiply_Float(A=0.03, B=1)
     )
     1=&MakeLinearColorFromVectorAndFloat(
      "Vector (RGB)"=&MakeVectorFromLinearColorRGB(Color=&MultiplyLinearColor(Linear Color=$Particles.Color, Scale Factor=&MakeVector4_FromLinearColor(LinearColor=$Particles.Color)))
      "Float (Alpha)"=&Multiply_Float(A=0.03, B=0.5)
     )
     2=?
     3=?
     Index=&SpawnGroup()
     Num=::NewEnumerator0
    )) {isSetParameters=true, module=SetVariables_#7}
   &Drag(
    Drag=&MergeHelper_FloatSwitch(0=2.5, 1=0.5, 2=?, 3=?, Index=&SpawnGroup(), Num=::NewEnumerator0)
    Use Linear Drag=true
    Use Rotational Drag=true
   )
   &SolveForcesAndVelocity(Clamp Velocity=false, Limit Acceleration=false)
  ]
  rendererClasses [&NiagaraSpriteRendererProperties]
  renderers [
   {rendererClass=&NiagaraSpriteRendererProperties}
   {rendererClass=&NiagaraSpriteRendererProperties, rendererIndex=1}
   {rendererClass=&NiagaraSpriteRendererProperties, rendererIndex=2}
   {rendererClass=&NiagaraSpriteRendererProperties, rendererIndex=3}
  ]
 }
 Motes_Flares_MeshRibbons {
  emitterUpdateScript [
   &EmitterState()
   &SpawnRate(Spawn Probability=0.9, SpawnRate=16, Use Spawn Probability=true)
   &SpawnRate(Spawn Probability=0.8, SpawnGroup=1, SpawnRate=7, Use Spawn Probability=true) {module=SpawnRate001}
   &SpawnBurst_Instantaneous(Spawn Group=2, Spawn Time=0.2, Use Spawn Probability=false, UseLoopCountLimit=true)
   &SpawnRate(
    SpawnGroup=2
    SpawnRate=&RandomRangeFloat(Minimum=0.8, Maximum=0.9, Evaluation Type=::NewEnumerator1, Randomness Mode=::NewEnumerator2, Random Seed=?)
    Use Spawn Probability=true
   ) {module=SpawnRate002}
   &SpawnBurst_Instantaneous(Spawn Group=3, Use Spawn Probability=false, UseLoopCountLimit=true) {module=SpawnBurst_Instantaneous001}
   &SpawnRate(
    SpawnGroup=3
    SpawnRate=&MakeCustomFloatFromBool(
     Bool=&SetBoolByFloatComparison(A=$Emitter.Age, B=0.45)
     True Float=&RandomRangeFloat(Minimum=0.8, Maximum=0.9, Evaluation Type=::NewEnumerator1, Randomness Mode=::NewEnumerator2, Random Seed=?)
     False Float=0
    )
    Use Spawn Probability=true
   ) {module=SpawnRate003}
  ]
  particleSpawnScript [
   /SetVariables(Particles.VisibilityTag=&SpawnGroup()) {isSetParameters=true, module=SetVariables_#4}
   &PartitionParticles() {module=PartitionParticles001}
   &PartitionParticles(Number Of Partitions=3, Partition Mode:ENiagara_PartitionMode_v2=::NewEnumerator5)
   &ShapeLocation(
    MaskSpawnGroup=false
    Non Uniform Scale=<1.2 1 1>
    Override Seed=false
    Rotation Coordinate Space=::Simulation
    Shape Origin=&VectorFromFloat(Value=0)
    Sphere Radius=&MergeHelper_FloatSwitch(0=15, 1=15, 2=0, 3=0, Index=&SpawnGroup(), Num=::NewEnumerator2)
    Sphere Surface Distribution=1
    UseOwnerScale=true
   )
   &InitializeParticle(
    Color=&MergeHelper_ColorSwitch(
     0=&RandomRangeLinearColor(Minimum=<rgba 1 0.4865 0.2632 1>, Maximum=<rgba 1 0.8182 0.0938 1>)
     1=&RandomRangeLinearColor(Minimum=<rgba 1 0.4865 0.2632 1>, Maximum=<rgba 1 0.8182 0.0938 1>)
     2=&MultiplyLinearColor(
      Linear Color=<rgba 1 0.853 0.6797 1>
      Scale Factor=&MakeVector4_FromLinearColor(LinearColor=$System.TODColor)
     )
     3=&MultiplyLinearColor(
      Linear Color=<rgba 1 0.853 0.6797 1>
      Scale Factor=&MakeVector4_FromLinearColor(LinearColor=$System.TODColor)
     )
     Index=&SpawnGroup()
     Num=::NewEnumerator2
    )
    Lifetime=&MergeHelper_FloatSwitch(
     0=&RandomRangeFloat(Minimum=1.2, Maximum=1.5)
     1=&RandomRangeFloat(Minimum=0.8, Maximum=1.5)
     2=&RandomRangeFloat(Minimum=1.25, Maximum=1.65)
     3=&RandomRangeFloat(Minimum=1.25, Maximum=1.65)
     Index=&SpawnGroup()
     Num=::NewEnumerator2
    )
    Mesh Scale=&Multiply_VectorByFloat(
     Vector=<0.1 0.1 0.1>
     Float=&Multiply_Float(
      A=&RandomRangeFloat(Minimum=0.7, Maximum=0.85)
      B=&MakeCustomFloatFromBool(Bool=$User.IsLargeBody, True Float=1.1, False Float=1)
     )
    )
    Mesh Scale Mode=::NewEnumerator4
    Sprite Rotation Mode=::NewEnumerator0
    Sprite Size Mode=::NewEnumerator3
    Uniform Sprite Size=&MergeHelper_FloatSwitch(
     0=&RandomRangeFloat(Minimum=1, Maximum=2)
     1=&RandomRangeFloat(Minimum=8, Maximum=15)
     2=?
     3=?
     Index=&SpawnGroup()
     Num=::NewEnumerator0
    )
    UsePositionOffset=false
    Write Material Random=true
   )
   &AddVelocity(
    Cone Angle=10
    Cone Axis=<0 0 1>
    Module.UseConeFalloff=false
    Module.UseVelDistribution=false
    Override Seed=false
    Rotation Coordinate Space=::World
    Velocity Mode=::NewEnumerator2
    Velocity Speed=&Multiply_Float(
     A=&MergeHelper_FloatSwitch(
      0=&RandomRangeFloat(Minimum=45, Maximum=60)
      1=&RandomRangeFloat(Minimum=15, Maximum=20)
      2=0
      3=0
      Index=&SpawnGroup()
      Num=::NewEnumerator2
     )
     B=&MakeCustomFloatFromBool(Bool=$User.IsLargeBody, True Float=1.3, False Float=1)
    )
   )
   &AddVelocity(
    Velocity=&MergeHelper_VectorSwitch(0=<0 0 0>, 1=<0 0 0>, 2=<0.52 0 1>, 3=<-0.52 0 1>, Index=&SpawnGroup(), Num=::NewEnumerator2)
    Velocity Speed Scale=&MergeHelper_FloatSwitch(
     0=0
     1=0
     2=&Multiply_Float(
      A=&RandomRangeFloat(Minimum=22, Maximum=27)
      B=&MakeCustomFloatFromBool(Bool=$User.IsLargeBody, True Float=1.15, False Float=1)
     )
     3=&Multiply_Float(
      A=&RandomRangeFloat(Minimum=22, Maximum=27)
      B=&MakeCustomFloatFromBool(Bool=$User.IsLargeBody, True Float=1.15, False Float=1)
     )
     Index=&SpawnGroup()
     Num=::NewEnumerator2
    )
   ) {module=AddVelocity001}
   &InitialMeshOrientation(
    Mesh Orientation Mode=::NewEnumerator2
    Orientation Vector=<0 0 1>
    Rotation=&RandomRangeVector(Minimum=<0 0 0>, Maximum=<0 0 1>)
    Use Rotation Vector=true
   )
   &CameraOffset(Camera Offset Amount=&MergeHelper_FloatSwitch(0=-25, 1=-25, 2=-10, 3=-10, Index=&SpawnGroup(), Num=::NewEnumerator2))
  ]
  particleUpdateScript [
   /SetVariables(Particles.FlickerWaveform:NiagaraFloat=&Waveform(
     Global Phase=&Add_Float(A=$Engine.Time, B=&RandomRangeFloat(Minimum=0, Maximum=1))
     "Waveform [1]"=::NewEnumerator10
     "[1] Frequency"=15
     "[1] Amplitude Min/Max"=<0.2 1.2>
    )) {isSetParameters=true, module=SetVariables_#5}
   &ParticleState()
   &UpdateMeshOrientation(
    Rotation Rate=&Multiply_Float(A=&RandomRangeFloat(Minimum=0.4, Maximum=0.85), B=1)
    Rotation Vector=&MakeVector(X=0, Y=0, Z=1)
   )
   &SkeletalMeshLocation(
    Bone / Socket Index=&MergeHelper_IntSwitch(0=7, 1=7, 2=1, 3=2, 4=?, 5=?, Index=&SpawnGroup(), Num=::NewEnumerator2)
    Bone Sampling Mode=::NewEnumerator4
    "Inherit Velocity (Scale)":NiagaraFloat=&MergeHelper_FloatSwitch(0=-0.07, 1=-0.07, 2=0, 3=0, Index=&SpawnGroup(), Num=::NewEnumerator2)
    Inherited Velocity Speed Limit=25
    MaskSpawnGroup=false
    Orientation Sampling=::NewEnumerator1
    Position Sampling=::NewEnumerator1
    Sampled Position Offset=&AddVectorToPosition(
     Position=Linked() {linkedVariable {name=Particles.Initial.ShapeLocation.ShapePosition, type=Vector3f}}
     Vector=&MergeHelper_VectorSwitch(
      0=&SelectVectorFromArray(
       Array Sampling Mode=::NewEnumerator0
       Direct Array Index=?
       Vector Selection Array=DataInterface() {propertyValues {FloatData [<10 0 0>, <-5 0 -14>, <-5 0 14>]}}
       Random Seed=0
      )
      1=&SelectVectorFromArray(
       Array Sampling Mode=::NewEnumerator0
       Direct Array Index=?
       Vector Selection Array=DataInterface() {propertyValues {FloatData [<20 0 0>, <10 0 -12>, <10 0 12>]}}
       Random Seed=0
      )
      2=&Lerp_Vector(A=<7 0 0>, B=<15 0 0>, Alpha=&RandomRangeFloat(Minimum=0, Maximum=1))
      3=&Lerp_Vector(A=<-7 0 0>, B=<-15 0 0>, Alpha=&RandomRangeFloat(Minimum=0, Maximum=1))
      Index=&SpawnGroup()
      Num=::NewEnumerator2
     )
    )
    Skeletal Mesh=$System.Skeletal Mesh
    Velocity Sampling=::NewEnumerator0
    VelocitySpeedLimit=true
   )
   &ScaleSpriteSize(Scale Factor=&MergeHelper_Vector2Switch(
     0=&Vector2DFromFloat(Value=&FloatFromCurve(
       FloatCurve=DataInterface() {propertyValues {Curve {keys [
           {arriveTangent=7.4528, interpMode=RCIM_Cubic, leaveTangent=7.4528, tangentMode=RCTM_User, time=0, value=0}
           <at 0.2 1.2 RCIM_Cubic>
           {arriveTangent=-2.7132, interpMode=RCIM_Cubic, leaveTangent=-2.7132, tangentMode=RCTM_User, time=1, value=0}
          ]}}}
       Scale Curve=&Multiply_Float(A=&FN_FrontEndFloatModifier(Gameplay=0.6, Front End=0.5), B=$Particles.FlickerWaveform)
      ))
     1=&Multiply_Vector2D_ByFloat(
      Float=&Multiply_Float(
       A=&FloatFromCurve(FloatCurve=DataInterface() {propertyValues {Curve {keys [
            {arriveTangent=7.4528, interpMode=RCIM_Cubic, leaveTangent=7.4528, tangentMode=RCTM_User, time=0, value=0}
            <at 0.2 1.2 RCIM_Cubic>
            {arriveTangent=-2.7132, interpMode=RCIM_Cubic, leaveTangent=-2.7132, tangentMode=RCTM_User, time=1, value=0}
           ]}}})
       B=&FN_Fit(
        Clamp:NiagaraBool=false
        Value:NiagaraFloat=$Particles.FlickerWaveform
        SourceMin:NiagaraFloat=0.2
        SourceMax:NiagaraFloat=1.2
        DestMin:NiagaraFloat=0.5
        DestMax:NiagaraFloat=1
       )
      )
      Vector2D=&MakeVector2D(
       X=&FloatFromCurve(FloatCurve=DataInterface() {propertyValues {Curve {keys [<at 0 0.8>,<at 1 1>]}}})
       Y=&FloatFromCurve(FloatCurve=DataInterface() {propertyValues {Curve {keys [<at 0 1.2529>,<at 1 1>]}}})
      )
     )
     2=?
     3=?
     4=?
     Index=&SpawnGroup()
     Num=::NewEnumerator0
    ))
   &ScaleMeshSize(Scale Factor=&Multiply_VectorByFloat(
     Vector=&MakeVector(
      X=&FloatFromCurve(FloatCurve=DataInterface() {propertyValues {Curve {keys [
           <at 0 1.2 RCIM_Cubic>
           {arriveTangent=-0.1903, interpMode=RCIM_Cubic, leaveTangent=-0.1903, time=0.8, value=0.929}
           {arriveTangent=0.6374, interpMode=RCIM_Cubic, leaveTangent=0.6374, tangentMode=RCTM_User, time=0.998, value=1.0101}
          ]}}})
      Y=&FloatFromCurve(FloatCurve=DataInterface() {propertyValues {Curve {keys [
           <at 0 1.2 RCIM_Cubic>
           {arriveTangent=-0.2178, interpMode=RCIM_Cubic, leaveTangent=-0.2178, time=0.8, value=0.9269}
           {arriveTangent=0.3322, interpMode=RCIM_Cubic, leaveTangent=0.3322, tangentMode=RCTM_User, time=0.9991, value=0.9824}
          ]}}})
      Z=&FloatFromCurve(
       FloatCurve=DataInterface() {propertyValues {Curve {keys [
           {arriveTangent=3.0747, interpMode=RCIM_Cubic, leaveTangent=3.0747, tangentMode=RCTM_User, time=0, value=0.6}
           {arriveTangent=0.5333, interpMode=RCIM_Cubic, leaveTangent=0.5333, time=0.15, value=0.85}
           {arriveTangent=0.4243, interpMode=RCIM_Cubic, leaveTangent=0.4243, time=0.75, value=1}
           {arriveTangent=0.9676, interpMode=RCIM_Cubic, leaveTangent=0.9676, tangentMode=RCTM_User, time=1, value=1.2106}
          ]}}}
       Scale Curve=&RandomRangeFloat(Minimum=0.9, Maximum=1.2)
      )
     )
     Float=&FloatFromCurve(FloatCurve=DataInterface() {propertyValues {Curve {
         defaultValue=1
         keys [
          {arriveTangent=3.7509, interpMode=RCIM_Cubic, leaveTangent=3.7509, tangentMode=RCTM_User, time=0, value=0.5}
          {arriveTangentWeight=0.1274, interpMode=RCIM_Cubic, leaveTangentWeight=0.7778, tangentMode=RCTM_User, time=0.2759, value=0.9}
          {arriveTangent=-0.0405, interpMode=RCIM_Cubic, leaveTangent=-0.0405, tangentMode=RCTM_User, time=1, value=1}
         ]
        }}})
    ))
   &CurlNoiseForce(
    Bias Noise=true
    Bias Noise Field:Vector3f=<0 0 0.25>
    Noise Strength=&FloatFromCurve(
     FloatCurve=DataInterface() {propertyValues {Curve {
        defaultValue=1
        keys [{interpMode=RCIM_Cubic, tangentMode=RCTM_User, time=0, value=0}, {interpMode=RCIM_Cubic, tangentMode=RCTM_User, time=1, value=1}]
       }}}
     Scale Curve=&MergeHelper_FloatSwitch(0=30, 1=30, 2=0, 3=0, Index=&SpawnGroup(), Num=::NewEnumerator2)
    )
    Pan Noise=true
    Pan Noise Field=<0 0 1>
   )
   &ScaleColor(
    Scale Alpha=&MergeHelper_FloatSwitch(
     0=1
     1=1
     2=&FloatFromCurve(FloatCurve=DataInterface() {propertyValues {Curve {
         defaultValue=1
         keys [
          {arriveTangent=1.3708, interpMode=RCIM_Cubic, leaveTangent=1.3708, tangentMode=RCTM_User, time=0, value=0}
          {arriveTangentWeight=0.0469, interpMode=RCIM_Cubic, leaveTangentWeight=0.2197, tangentMode=RCTM_User, time=0.1, value=1}
          {arriveTangentWeight=0.3342, interpMode=RCIM_Cubic, leaveTangentWeight=0.1014, tangentMode=RCTM_User, time=0.6047, value=1}
          {arriveTangent=-1.7747, interpMode=RCIM_Cubic, leaveTangent=-1.7747, tangentMode=RCTM_User, time=1, value=0}
         ]
        }}})
     3=&FloatFromCurve(FloatCurve=DataInterface() {propertyValues {Curve {
         defaultValue=1
         keys [
          {arriveTangent=1.3708, interpMode=RCIM_Cubic, leaveTangent=1.3708, tangentMode=RCTM_User, time=0, value=0}
          {arriveTangentWeight=0.0469, interpMode=RCIM_Cubic, leaveTangentWeight=0.2197, tangentMode=RCTM_User, time=0.1, value=1}
          {arriveTangentWeight=0.3342, interpMode=RCIM_Cubic, leaveTangentWeight=0.1014, tangentMode=RCTM_User, time=0.6047, value=1}
          {arriveTangent=-1.7747, interpMode=RCIM_Cubic, leaveTangent=-1.7747, tangentMode=RCTM_User, time=1, value=0}
         ]
        }}})
     Index=&SpawnGroup()
     Num=::NewEnumerator2
    )
    Scale RGB=&MergeHelper_VectorSwitch(
     0=&Multiply_VectorByFloat(
      Vector=&MakeVectorFromLinearColorRGB(Color=&ColorFromCurve(
        ColorCurve=DataInterface() {propertyValues {
          AlphaCurve {keys [<at 0 1 RCIM_Cubic>,<at 1 0 RCIM_Cubic>]}
          BlueCurve {keys [<at 0 1 RCIM_Cubic>,<at 1 1 RCIM_Cubic>]}
          GreenCurve {keys [<at 0 1 RCIM_Cubic>,<at 1 1 RCIM_Cubic>]}
          RedCurve {keys [<at 0 1 RCIM_Cubic>,<at 1 1 RCIM_Cubic>]}
         }}
        CurveIndex=&Frac_Float(Float=&Add_Float(A=&RandomRangeFloat(Minimum=0, Maximum=1), B=$Particles.NormalizedAge))
       ))
      Float=&FloatFromCurve(
       FloatCurve=DataInterface() {propertyValues {Curve {
          defaultValue=1
          keys [
           {interpMode=RCIM_Cubic, tangentMode=RCTM_User, time=0, value=1}
           {arriveTangent=-2.2239, interpMode=RCIM_Cubic, leaveTangent=-2.2239, tangentMode=RCTM_User, time=1, value=0.2}
          ]
         }}}
       Scale Curve=&Multiply_Float(
        A=4
        B=&FN_Fit(
         Clamp:NiagaraBool=false
         Value:NiagaraFloat=$Particles.FlickerWaveform
         SourceMin:NiagaraFloat=0.2
         SourceMax:NiagaraFloat=1.2
         DestMin:NiagaraFloat=0.75
         DestMax:NiagaraFloat=1.2
        )
       )
      )
     )
     1=&Multiply_VectorByFloat(
      Vector=&MakeVectorFromLinearColorRGB(Color=&ColorFromCurve(
        ColorCurve=DataInterface() {propertyValues {
          AlphaCurve {keys [<at 0 1 RCIM_Cubic>,<at 1 0 RCIM_Cubic>]}
          BlueCurve {keys [<at 0 1 RCIM_Cubic>,<at 1 1 RCIM_Cubic>]}
          GreenCurve {keys [<at 0 1 RCIM_Cubic>,<at 1 1 RCIM_Cubic>]}
          RedCurve {keys [<at 0 1 RCIM_Cubic>,<at 1 1 RCIM_Cubic>]}
         }}
        CurveIndex=&Frac_Float(Float=&Add_Float(A=&RandomRangeFloat(Minimum=0, Maximum=1), B=$Particles.NormalizedAge))
       ))
      Float=&FloatFromCurve(
       FloatCurve=DataInterface() {propertyValues {Curve {
          defaultValue=1
          keys [
           {interpMode=RCIM_Cubic, tangentMode=RCTM_User, time=0, value=1}
           {arriveTangent=-2.2239, interpMode=RCIM_Cubic, leaveTangent=-2.2239, tangentMode=RCTM_User, time=1, value=0.2}
          ]
         }}}
       Scale Curve=&Multiply_Float(
        A=4
        B=&FN_Fit(
         Clamp:NiagaraBool=false
         Value:NiagaraFloat=$Particles.FlickerWaveform
         SourceMin:NiagaraFloat=0.2
         SourceMax:NiagaraFloat=1.2
         DestMin:NiagaraFloat=0.75
         DestMax:NiagaraFloat=1.2
        )
       )
      )
     )
     2=&MakeVectorFromLinearColorRGB(Color=&ColorFromCurve(ColorCurve=DataInterface() {propertyValues {
         AlphaCurve {keys [<at 0 1 RCIM_Cubic>]}
         BlueCurve {keys [<at 0 2.125>, {arriveTangent=-1.125, interpMode=RCIM_Cubic, leaveTangent=-1.125, time=0.1778, value=1.0156}, <at 1 1 RCIM_Cubic>]}
         GreenCurve {keys [<at 0 2.7618>, {arriveTangent=-1.7618, interpMode=RCIM_Cubic, leaveTangent=-1.7618, time=0.1778, value=1.7965}, <at 1 1 RCIM_Cubic>]}
         RedCurve {keys [<at 0 4>, {arriveTangent=-3, interpMode=RCIM_Cubic, leaveTangent=-3, time=0.1778, value=3}, <at 1 1 RCIM_Cubic>]}
        }}))
     3=&MakeVectorFromLinearColorRGB(Color=&ColorFromCurve(ColorCurve=DataInterface() {propertyValues {
         AlphaCurve {keys [<at 0 1 RCIM_Cubic>]}
         BlueCurve {keys [<at 0 2.125>, {arriveTangent=-1.125, interpMode=RCIM_Cubic, leaveTangent=-1.125, time=0.1778, value=1.0156}, <at 1 1 RCIM_Cubic>]}
         GreenCurve {keys [<at 0 2.7618>, {arriveTangent=-1.7618, interpMode=RCIM_Cubic, leaveTangent=-1.7618, time=0.1778, value=1.7965}, <at 1 1 RCIM_Cubic>]}
         RedCurve {keys [<at 0 4>, {arriveTangent=-3, interpMode=RCIM_Cubic, leaveTangent=-3, time=0.1778, value=3}, <at 1 1 RCIM_Cubic>]}
        }}))
     Index=&SpawnGroup()
     Num=::NewEnumerator2
    )
    ScaleA=true
    ScaleRGB=true
   )
   /SetVariables(Particles.Glow03_Size=&MergeHelper_Vector2Switch(
     0=&Multiply_Vector2D_ByFloat(
      Float=&FloatFromCurve(
       FloatCurve=DataInterface() {propertyValues {Curve {keys [
           <at 0 0 RCIM_Cubic>
           <at 0.2 1.3 RCIM_Cubic>
           {arriveTangent=-0.1423, interpMode=RCIM_Cubic, leaveTangent=-0.1423, tangentMode=RCTM_User, time=1, value=0}
          ]}}}
       Scale Curve=&Multiply_Float(A=&MakeFloatFromVector2d(Vector 2D=$Particles.SpriteSize), B=3.5)
      )
      Vector2D=<1 1.5>
     )
     1=&Multiply_Vector2D_ByFloat(
      Float=&FloatFromCurve(
       FloatCurve=DataInterface() {propertyValues {Curve {keys [
           <at 0 0 RCIM_Cubic>
           <at 0.2 1.3 RCIM_Cubic>
           {arriveTangent=-0.1423, interpMode=RCIM_Cubic, leaveTangent=-0.1423, tangentMode=RCTM_User, time=1, value=0}
          ]}}}
       Scale Curve=&Multiply_Float(A=&MakeFloatFromVector2d(Vector 2D=$Particles.SpriteSize), B=2)
      )
      Vector2D=<1 1>
     )
     2=?
     3=?
     4=?
     Index=&SpawnGroup()
     Num=::NewEnumerator0
    )) {isSetParameters=true, module=SetVariables_#6}
   /SetVariables(Particles.Glow03_Color=&MergeHelper_ColorSwitch(
     0=&MakeLinearColorFromVectorAndFloat(
      "Vector (RGB)"=&MakeVectorFromLinearColorRGB(Color=&MultiplyLinearColor(Linear Color=$Particles.Color, Scale Factor=&MakeVector4_FromLinearColor(LinearColor=$Particles.Color)))
      "Float (Alpha)"=&Multiply_Float(A=0.03, B=1)
     )
     1=&MakeLinearColorFromVectorAndFloat(
      "Vector (RGB)"=&MakeVectorFromLinearColorRGB(Color=&MultiplyLinearColor(Linear Color=$Particles.Color, Scale Factor=&MakeVector4_FromLinearColor(LinearColor=$Particles.Color)))
      "Float (Alpha)"=&Multiply_Float(A=0.03, B=0.5)
     )
     2=?
     3=?
     Index=&SpawnGroup()
     Num=::NewEnumerator0
    )) {isSetParameters=true, module=SetVariables_#7}
   &Drag(
    Drag=&MergeHelper_FloatSwitch(
     0=2.5
     1=0.5
     2=&RandomRangeFloat(Minimum=0.8, Maximum=1.1)
     3=&RandomRangeFloat(Minimum=0.8, Maximum=1.1)
     Index=&SpawnGroup()
     Num=::NewEnumerator2
    )
    Use Linear Drag=true
    Use Rotational Drag=true
   )
   &SolveForcesAndVelocity(Clamp Velocity=false, Limit Acceleration=false)
   /SetVariables("Particles.MeshDyn 0 ":Vector4f=&MakeVector4(
     X=&FloatFromCurve(
      FloatCurve=DataInterface() {propertyValues {Curve {keys [
          <at 0.3 0.2 RCIM_Cubic>
          {arriveTangent=0.5424, interpMode=RCIM_Cubic, leaveTangent=0.5424, tangentMode=RCTM_User, time=0.605, value=0.6745}
          <at 1.0013 0.7537 RCIM_Cubic>
         ]}}}
      Scale Curve=&Lerp_Float(A=0.7, B=0.8, Alpha=&RandomRangeFloat(Minimum=0, Maximum=1))
     )
     Y=&RandomRangeFloat(Minimum=0, Maximum=1)
     Z=&FloatFromCurve(FloatCurve=DataInterface() {propertyValues {Curve {keys [
          <at 0.6 1 RCIM_Cubic>
          {arriveTangent=-0.6876, arriveTangentWeight=0.2346, interpMode=RCIM_Cubic, leaveTangent=-0.6876, tangentMode=RCTM_User, time=1, value=0.4}
         ]}}})
     W=&FloatFromCurve(FloatCurve=DataInterface() {propertyValues {Curve {
         defaultValue=1
         keys [{interpMode=RCIM_Cubic, tangentMode=RCTM_User, time=0.6, value=0.01}, {interpMode=RCIM_Cubic, tangentMode=RCTM_User, time=1, value=0.05}]
        }}})
    )) {isSetParameters=true}
  ]
  rendererClasses [&NiagaraSpriteRendererProperties, &NiagaraMeshRendererProperties]
  renderers [
   {rendererClass=&NiagaraSpriteRendererProperties}
   {rendererClass=&NiagaraSpriteRendererProperties, rendererIndex=1}
   {rendererClass=&NiagaraSpriteRendererProperties, rendererIndex=2}
   {rendererClass=&NiagaraSpriteRendererProperties, rendererIndex=3}
   {rendererClass=&NiagaraMeshRendererProperties, rendererIndex=4}
   {rendererClass=&NiagaraMeshRendererProperties, rendererIndex=5}
  ]
 }
 Motes_From_Character {
  emitterUpdateScript [
   &EmitterState()
   &SpawnBurst_Instantaneous(Spawn Count=6, Use Spawn Probability=false, UseLoopCountLimit=true)
   &SpawnRate(Spawn Probability=0.9, SpawnRate=15, Use Spawn Probability=true)
  ]
  particleSpawnScript [
   /SetVariables(Particles.Lerp Random:NiagaraFloat=&FloatFromCurve(
     FloatCurve=DataInterface() {propertyValues {Curve {keys [<at 0 1>,<at 0.7 0.1561>,<at 1 0>]}}}
     Scale Curve=&RandomRangeFloat(Minimum=0, Maximum=1)
    )) {isSetParameters=true}
   &InitializeParticle(
    Color=&RandomRangeLinearColor(Minimum=<rgba 1 0.4361 0.1908 0.2>, Maximum=<rgba 1 0.7719 0.0808 1>)
    Lifetime=&Lerp_Float(A=1.45, B=1, Alpha=$Particles.Lerp Random)
    Sprite Rotation Mode=::NewEnumerator0
    Sprite Size Mode=::NewEnumerator3
    Uniform Sprite Size=&Multiply_Float(
     A=&Lerp_Float(A=1, B=2, Alpha=$Particles.Lerp Random)
     B=&RandomRangeFloat(Minimum=1, Maximum=1.2)
    )
    UsePositionOffset=false
    Write Material Random=true
   )
   &CameraOffset(Camera Offset Amount=-5)
   &PartitionParticles(Number Of Partitions=4)
  ]
  particleUpdateScript [
   /SetVariables(Particles.FlickerWaveform:NiagaraFloat=&Waveform(
     Global Phase=&Add_Float(A=$Engine.Time, B=&RandomRangeFloat(Minimum=0, Maximum=1))
     "Waveform [1]"=::NewEnumerator10
     "[1] Frequency"=5
     "[1] Amplitude Min/Max"=<0.2 1.2>
    )) {isSetParameters=true, module=SetVariables_#5}
   &SkeletalMeshLocation(
    Bone / Socket Index=&SelectIntFromArray(Direct Array Index=$Particles.PartitionParticles.Partition, Int Selection Array=DataInterface() {propertyValues {IntData [7,7,1,2]}})
    Bone Sampling Mode=::NewEnumerator4
    "Inherit Velocity (Scale)"=-0.05
    Inherited Velocity Speed Limit=35
    MaskSpawnGroup=false
    Orientation Sampling=::NewEnumerator1
    Sampled Position Offset=&MergeHelper_VectorSwitch(
     0=&Lerp_Vector(A=<7 -5 -6>, B=<5 0 -19>, Alpha=$Particles.Lerp Random)
     1=&Lerp_Vector(A=<7 -5 6>, B=<5 0 19>, Alpha=$Particles.Lerp Random)
     2=&Lerp_Vector(A=<18 0 0>, B=<5 0 0>, Alpha=$Particles.Lerp Random)
     3=&Lerp_Vector(A=<-18 0 0>, B=<-5 0 0>, Alpha=$Particles.Lerp Random)
     Index=$Particles.PartitionParticles.Partition
     Num=::NewEnumerator2
    )
    Skeletal Mesh=$System.Skeletal Mesh
    Use Sampled Mesh Attribute as Ribbon ID=false
    Velocity Sampling=::NewEnumerator0
    VelocitySpeedLimit=true
   )
   &ParticleState()
   &ScaleSpriteSize(Scale Factor=&Multiply_Vector2D_ByFloat(
     Float=&Multiply_Float(
      A=&FloatFromCurve(FloatCurve=DataInterface() {propertyValues {Curve {keys [
           <at 0 1 RCIM_Cubic>
           {arriveTangent=-1.1725, arriveTangentWeight=0.7014, interpMode=RCIM_Cubic, leaveTangent=-1.1725, tangentMode=RCTM_User, time=1, value=0}
          ]}}})
      B=&FN_Fit(
       Clamp:NiagaraBool=false
       Value:NiagaraFloat=$Particles.FlickerWaveform
       SourceMin:NiagaraFloat=0.2
       SourceMax:NiagaraFloat=1.2
       DestMin:NiagaraFloat=0.6
       DestMax:NiagaraFloat=1
      )
     )
     Vector2D=&MakeVector2D(
      X=&FloatFromCurve(FloatCurve=DataInterface() {propertyValues {Curve {keys [<at 0 0.7 RCIM_Cubic>, {arriveTangent=0.3478, interpMode=RCIM_Cubic, leaveTangent=0.3478, tangentMode=RCTM_User, time=1, value=1}]}}})
      Y=&FloatFromCurve(FloatCurve=DataInterface() {propertyValues {Curve {keys [
           {arriveTangent=0.0224, interpMode=RCIM_Cubic, leaveTangent=0.0224, tangentMode=RCTM_User, time=0, value=1.2}
           {arriveTangent=-0.3923, interpMode=RCIM_Cubic, leaveTangent=-0.3923, tangentMode=RCTM_User, time=1, value=1}
          ]}}})
     )
    ))
   &ScaleColor(
    Scale Alpha=&FloatFromCurve(FloatCurve=DataInterface() {propertyValues {Curve {
        defaultValue=1
        keys [
         {interpMode=RCIM_Cubic, tangentMode=RCTM_User, time=0, value=1}
         {arriveTangentWeight=0.8128, interpMode=RCIM_Cubic, tangentMode=RCTM_User, time=1, value=0.5}
        ]
       }}})
    Scale RGB=&Multiply_VectorByFloat(
     Vector=&MakeVectorFromLinearColorRGB(Color=<rgba 1 1 1 1>)
     Float=&FloatFromCurve(
      FloatCurve=DataInterface() {propertyValues {Curve {
         defaultValue=1
         keys [
          {interpMode=RCIM_Cubic, tangentMode=RCTM_User, time=0, value=1}
          {arriveTangentWeight=0.8128, interpMode=RCIM_Cubic, tangentMode=RCTM_User, time=1, value=0.5}
         ]
        }}}
      Scale Curve=&Multiply_Float(
       A=1.5
       B=&FN_Fit(
        Clamp:NiagaraBool=false
        Value:NiagaraFloat=$Particles.FlickerWaveform
        SourceMin:NiagaraFloat=0.2
        SourceMax:NiagaraFloat=1.2
        DestMin:NiagaraFloat=0.5
        DestMax:NiagaraFloat=3
       )
      )
     )
    )
    ScaleA=true
    ScaleRGB=true
   )
   /SetVariables(Particles.Glow03_Size=&Multiply_Vector2D_ByFloat(
     Float=&FloatFromCurve(
      FloatCurve=DataInterface() {propertyValues {Curve {keys [
          <at 0 0 RCIM_Cubic>
          <at 0.2 1.3 RCIM_Cubic>
          {arriveTangent=-0.1423, interpMode=RCIM_Cubic, leaveTangent=-0.1423, tangentMode=RCTM_User, time=1, value=0}
         ]}}}
      Scale Curve=&Multiply_Float(A=&MakeFloatFromVector2d(Vector 2D=$Particles.SpriteSize), B=2.5)
     )
     Vector2D=<1 2>
    )) {isSetParameters=true, module=SetVariables_#6}
   /SetVariables(Particles.Glow03_Color=&MergeHelper_ColorSwitch(
     0=&MakeLinearColorFromVectorAndFloat(
      "Vector (RGB)"=&MakeVectorFromLinearColorRGB(Color=&MultiplyLinearColor(Linear Color=$Particles.Color, Scale Factor=&MakeVector4_FromLinearColor(LinearColor=$Particles.Color)))
      "Float (Alpha)"=&Multiply_Float(A=0.1, B=1)
     )
     1=&MakeLinearColorFromVectorAndFloat(
      "Vector (RGB)"=&MakeVectorFromLinearColorRGB(Color=&MultiplyLinearColor(Linear Color=$Particles.Color, Scale Factor=&MakeVector4_FromLinearColor(LinearColor=$Particles.Color)))
      "Float (Alpha)"=&Multiply_Float(A=0.03, B=0.5)
     )
     2=?
     3=?
     Index=&SpawnGroup()
     Num=::NewEnumerator0
    )) {isSetParameters=true, module=SetVariables_#7}
   /SetVariables(Particles.PositionOffset:NiagaraPosition=&AddVectorToPosition(
     Position=$Particles.Position
     Vector=&MakeVector(
      X=0
      Y=0
      Z=&Multiply_Float(
       A=&FloatFromCurve(
        FloatCurve=DataInterface() {propertyValues {Curve {
           defaultValue=1
           keys [
            {arriveTangent=3.1783, interpMode=RCIM_Cubic, leaveTangent=3.1783, tangentMode=RCTM_User, time=0, value=0}
            {
             arriveTangent=1.6217
             arriveTangentWeight=0.0843
             interpMode=RCIM_Cubic
             leaveTangent=1.6217
             leaveTangentWeight=0.3871
             tangentMode=RCTM_User
             time=0.2361
             value=0.565
            }
            {arriveTangent=0.5492, interpMode=RCIM_Cubic, leaveTangent=0.5492, tangentMode=RCTM_User, time=1, value=1}
           ]
          }}}
        Scale Curve=&Multiply_Float(
         A=&Lerp_Float(A=17, B=9, Alpha=$Particles.Lerp Random)
         B=&RandomRangeFloat(Minimum=1, Maximum=1.35)
        )
       )
       B=&MakeCustomFloatFromBool(Bool=$User.IsLargeBody, True Float=1.2, False Float=1)
      )
     )
    )) {isSetParameters=true}
  ]
  rendererClasses [&NiagaraSpriteRendererProperties]
  renderers [
   {rendererClass=&NiagaraSpriteRendererProperties}
   {rendererClass=&NiagaraSpriteRendererProperties, rendererIndex=1}
  ]
 }
 RibbonParticles002 {
  emitterUpdateScript [
   &EmitterState()
   &SpawnParticlesFromOtherEmitter(
    Attribute Reader=DataInterface() {propertyValues {EmitterBinding {bindingMode=Other, emitterName=Ribbons_From_Character001}}}
    Maximum Age Spawn Threshold=2.49
    Spawn Rate=&RandomRangeFloat(Minimum=5, Maximum=10)
    Spawn Rate Per Particle Cap=200
    Use Maximum Age=false
    Use Minimum Age=false
   )
  ]
  particleSpawnScript [
   &InitializeParticle(
    Color=&RandomRangeLinearColor(Minimum=<rgba 1 0.4865 0.2632 1>, Maximum=<rgba 1.5 1.2272 0.1406 1>)
    Lifetime=&RandomRangeFloat(Minimum=0.2, Maximum=0.7)
    Sprite Size Mode=::NewEnumerator3
    Uniform Sprite Size=&RandomRangeFloat(Minimum=0.5, Maximum=1)
    UsePositionOffset=false
    Write Material Random=true
   )
   &SampleParticlesFromOtherEmitter(
    Applied Velocity Scale=-0.1
    Maximum Sampled Particle Age=2.4
    Minimum Sampled Particle Age=2.1
    Module.MaxSampledAgeCutoff=true
    Module.MinSampledAgeCutoff=true
    Reject Particle By Sampled Age=true
    Reject Particle on Sampled Particle's First Frame=true
    Velocity Sampling=::NewEnumerator0
   )
   &ShapeLocation(MaskSpawnGroup=false, Override Seed=false, Sphere Radius=3, UseOwnerScale=true)
   &AddVelocity(
    Velocity=&RandomRangeVector(Minimum=<-0.2 -0.2 0>, Maximum=<0.2 0.2 1>)
    Velocity Speed Scale=&RandomRangeFloat(Minimum=18, Maximum=30)
   )
   &InheritVelocity(
    Inherited Velocity Amount Scale=&Multiply_VectorByFloat(Vector=<1 1 1>, Float=-1)
    Inherited Velocity Speed Limit=25
    Speed Limit=true
    UseSpeedThreshold=false
   )
   &CameraOffset(Camera Offset Amount=-10)
  ]
  particleUpdateScript [
   &ParticleState()
   /SetVariables(Particles.FlickerWaveform:NiagaraFloat=&Waveform(
     Global Phase=&Add_Float(A=$Engine.Time, B=&RandomRangeFloat(Minimum=0, Maximum=1))
     "Waveform [1]"=::NewEnumerator10
     "[1] Frequency"=10
     "[1] Amplitude Min/Max"=<0.2 1.1>
    )) {isSetParameters=true, module=SetVariables_#8}
   &ScaleSpriteSize(Scale Factor=&Vector2DFromFloat(Value=&FloatFromCurve(
      FloatCurve=DataInterface() {propertyValues {Curve {keys [
          {arriveTangent=7.4528, interpMode=RCIM_Cubic, leaveTangent=7.4528, tangentMode=RCTM_User, time=0, value=0.5}
          {arriveTangent=-0.5, interpMode=RCIM_Cubic, leaveTangent=-0.5, time=0.2, value=1.2}
          {arriveTangent=-2.7132, interpMode=RCIM_Cubic, leaveTangent=-2.7132, tangentMode=RCTM_User, time=1, value=0}
         ]}}}
      Scale Curve=&Multiply_Float(A=&FN_FrontEndFloatModifier(Gameplay=1, Front End=0.8), B=$Particles.FlickerWaveform)
     )))
   &ScaleColor(
    Scale Alpha=&FloatFromCurve(FloatCurve=DataInterface() {propertyValues {Curve {
        defaultValue=1
        keys [
         {arriveTangentWeight=0.0589, interpMode=RCIM_Cubic, leaveTangentWeight=0.3596, tangentMode=RCTM_User, time=0, value=0.9847}
         {arriveTangent=-1.0221, interpMode=RCIM_Cubic, leaveTangent=-1.0221, tangentMode=RCTM_User, time=1, value=0}
        ]
       }}})
    Scale RGB=&Multiply_VectorByFloat(
     Vector=&MakeVectorFromLinearColorRGB(Color=&ColorFromCurve(
       ColorCurve=DataInterface() {propertyValues {
         AlphaCurve {keys [<at 0 1 RCIM_Cubic>,<at 1 0 RCIM_Cubic>]}
         BlueCurve {keys [<at 0 1 RCIM_Cubic>,<at 1 1 RCIM_Cubic>]}
         GreenCurve {keys [<at 0 1 RCIM_Cubic>,<at 1 1 RCIM_Cubic>]}
         RedCurve {keys [<at 0 1 RCIM_Cubic>,<at 1 1 RCIM_Cubic>]}
        }}
       CurveIndex=&Frac_Float(Float=&Add_Float(A=&RandomRangeFloat(Minimum=0, Maximum=1), B=$Particles.NormalizedAge))
      ))
     Float=&FloatFromCurve(
      FloatCurve=DataInterface() {propertyValues {Curve {
         defaultValue=1
         keys [
          {interpMode=RCIM_Cubic, tangentMode=RCTM_User, time=0, value=1}
          {arriveTangentWeight=0.8128, interpMode=RCIM_Cubic, tangentMode=RCTM_User, time=1, value=0.2}
         ]
        }}}
      Scale Curve=&Multiply_Float(
       A=4
       B=&FN_Fit(
        Clamp:NiagaraBool=false
        Value:NiagaraFloat=$Particles.FlickerWaveform
        SourceMin:NiagaraFloat=0.2
        SourceMax:NiagaraFloat=1.2
        DestMin:NiagaraFloat=0.75
        DestMax:NiagaraFloat=1.2
       )
      )
     )
    )
    ScaleA=true
    ScaleRGB=true
   ) {module=ScaleColor001}
   /SetVariables(Particles.Glow03_Color=&MakeLinearColorFromVectorAndFloat(
     "Vector (RGB)"=&MakeVectorFromLinearColorRGB(Color=&MultiplyLinearColor(Linear Color=$Particles.Color, Scale Factor=&MakeVector4_FromLinearColor(LinearColor=$Particles.Color)))
     "Float (Alpha)"=&Multiply_Float(A=0.03, B=1)
    )) {isSetParameters=true, module=SetVariables_#9}
   /SetVariables(Particles.Glow03_Size=&Multiply_Vector2D_ByFloat(
     Float=&FloatFromCurve(
      FloatCurve=DataInterface() {propertyValues {Curve {keys [
          <at 0 0 RCIM_Cubic>
          <at 0.2 1.3 RCIM_Cubic>
          {arriveTangent=-0.1423, interpMode=RCIM_Cubic, leaveTangent=-0.1423, tangentMode=RCTM_User, time=1, value=0}
         ]}}}
      Scale Curve=&Multiply_Float(A=&MakeFloatFromVector2d(Vector 2D=$Particles.SpriteSize), B=3.5)
     )
     Vector2D=<1 1.5>
    )) {isSetParameters=true, module=SetVariables_#10}
   &Drag(Drag=2, Use Linear Drag=true, Use Rotational Drag=true)
   &SolveForcesAndVelocity(Clamp Velocity=false, Limit Acceleration=false)
  ]
  rendererClasses [&NiagaraSpriteRendererProperties]
  renderers [
   {rendererClass=&NiagaraSpriteRendererProperties}
   {rendererClass=&NiagaraSpriteRendererProperties, rendererIndex=1}
  ]
 }
 RibbonSparkles002 {
  emitterUpdateScript [
   &EmitterState()
   &SpawnParticlesFromOtherEmitter(
    Attribute Reader=DataInterface() {propertyValues {EmitterBinding {bindingMode=Other, emitterName=Ribbons_From_Character001}}}
    Calculate Spawn Rate Per Particle=false
    Maximum Age Spawn Threshold=2
    Minimum Age Spawn Threshold=0.5
    Spawn Rate=&RandomRangeFloat(Minimum=2, Maximum=5)
    Use Maximum Age=false
    Use Minimum Age=false
   )
  ]
  particleSpawnScript [
   &InitializeParticle(
    Color=&RandomRangeLinearColor(Minimum=<rgba 1.5 0.9828 0.7578 0.2>, Maximum=<rgba 2 1.767 0.8385 1>)
    Lifetime=&RandomRangeFloat(Minimum=0.25, Maximum=0.35)
    Sprite Size Mode=::NewEnumerator3
    Uniform Sprite Size=&RandomRangeFloat(Minimum=11, Maximum=15)
    UsePositionOffset=false
    Write Material Random=true
   )
   &SampleParticlesFromOtherEmitter(
    Maximum Sampled Particle Age=2.3
    Minimum Sampled Particle Age=0.5
    Module.MaxSampledAgeCutoff=true
    Module.MinSampledAgeCutoff=true
    Reject Particle By Sampled Age=true
   )
   &ShapeLocation(MaskSpawnGroup=false, Override Seed=false, Sphere Radius=1.5, UseOwnerScale=true)
   &AddVelocity(
    Velocity=&RandomRangeVector(Minimum=<-0.2 -0.2 0>, Maximum=<0.2 0.2 1>)
    Velocity Speed Scale=&RandomRangeFloat(Minimum=18, Maximum=30)
   )
   &InheritVelocity(
    Inherited Velocity Amount Scale=&Multiply_VectorByFloat(Vector=<1 1 1>, Float=-1)
    Inherited Velocity Speed Limit=25
    Speed Limit=true
    UseSpeedThreshold=false
   )
   &CameraOffset(Camera Offset Amount=-5)
  ]
  particleUpdateScript [
   &ParticleState()
   /SetVariables(Particles.FlickerWaveform:NiagaraFloat=&Waveform(
     Global Phase=&Add_Float(A=$Engine.Time, B=&RandomRangeFloat(Minimum=0, Maximum=1))
     "Waveform [1]"=::NewEnumerator10
     "[1] Frequency"=10
     "[1] Amplitude Min/Max"=<0.5 1.1>
    )) {isSetParameters=true, module=SetVariables_#8}
   &ScaleSpriteSize(Scale Factor=&Vector2DFromFloat(Value=&FloatFromCurve(
      FloatCurve=DataInterface() {propertyValues {Curve {keys [
          {arriveTangent=-0.5231, interpMode=RCIM_Cubic, time=0.2, value=1}
          {arriveTangent=-2.7132, interpMode=RCIM_Cubic, leaveTangent=-2.7132, tangentMode=RCTM_User, time=1, value=0}
         ]}}}
      Scale Curve=&Multiply_Float(A=&FN_FrontEndFloatModifier(Gameplay=1, Front End=0.8), B=$Particles.FlickerWaveform)
     )))
   &ScaleColor(
    Scale Alpha=&FloatFromCurve(FloatCurve=DataInterface() {propertyValues {Curve {
        defaultValue=1
        keys [
         {arriveTangentWeight=0.0589, interpMode=RCIM_Cubic, leaveTangentWeight=0.3596, tangentMode=RCTM_User, time=0, value=0.9847}
         {arriveTangent=-1.0221, interpMode=RCIM_Cubic, leaveTangent=-1.0221, tangentMode=RCTM_User, time=1, value=0}
        ]
       }}})
    Scale RGB=&Multiply_VectorByFloat(
     Vector=&MakeVectorFromLinearColorRGB(Color=&ColorFromCurve(
       ColorCurve=DataInterface() {propertyValues {
         AlphaCurve {keys [<at 0 1 RCIM_Cubic>,<at 1 0 RCIM_Cubic>]}
         BlueCurve {keys [<at 0 1 RCIM_Cubic>,<at 1 1 RCIM_Cubic>]}
         GreenCurve {keys [<at 0 1 RCIM_Cubic>,<at 1 1 RCIM_Cubic>]}
         RedCurve {keys [<at 0 1 RCIM_Cubic>,<at 1 1 RCIM_Cubic>]}
        }}
       CurveIndex=&Frac_Float(Float=&Add_Float(A=&RandomRangeFloat(Minimum=0, Maximum=1), B=$Particles.NormalizedAge))
      ))
     Float=&FloatFromCurve(
      FloatCurve=DataInterface() {propertyValues {Curve {
         defaultValue=1
         keys [
          {interpMode=RCIM_Cubic, tangentMode=RCTM_User, time=0, value=1}
          {arriveTangentWeight=0.8128, interpMode=RCIM_Cubic, tangentMode=RCTM_User, time=1, value=0.2}
         ]
        }}}
      Scale Curve=&Multiply_Float(
       A=2
       B=&FN_Fit(
        Clamp:NiagaraBool=false
        Value:NiagaraFloat=$Particles.FlickerWaveform
        SourceMin:NiagaraFloat=0.5
        SourceMax:NiagaraFloat=1.2
        DestMin:NiagaraFloat=0.75
        DestMax:NiagaraFloat=1.2
       )
      )
     )
    )
    ScaleA=true
    ScaleRGB=true
   ) {module=ScaleColor001}
   /SetVariables(Particles.Glow03_Color=&MakeLinearColorFromVectorAndFloat(
     "Vector (RGB)"=&MakeVectorFromLinearColorRGB(Color=&MultiplyLinearColor(Linear Color=$Particles.Color, Scale Factor=&MakeVector4_FromLinearColor(LinearColor=$Particles.Color)))
     "Float (Alpha)"=&Multiply_Float(A=0.07, B=1)
    )) {isSetParameters=true, module=SetVariables_#9}
   /SetVariables(Particles.Glow03_Size=&Multiply_Vector2D_ByFloat(
     Float=&FloatFromCurve(
      FloatCurve=DataInterface() {propertyValues {Curve {keys [
          <at 0 0 RCIM_Cubic>
          <at 0.2 1.3 RCIM_Cubic>
          {arriveTangent=-0.1423, interpMode=RCIM_Cubic, leaveTangent=-0.1423, tangentMode=RCTM_User, time=1, value=0}
         ]}}}
      Scale Curve=&Multiply_Float(A=&MakeFloatFromVector2d(Vector 2D=$Particles.SpriteSize), B=0.5)
     )
     Vector2D=<1 1.5>
    )) {isSetParameters=true, module=SetVariables_#10}
   &Drag(Drag=2, Use Linear Drag=true, Use Rotational Drag=true)
   &SolveForcesAndVelocity(Clamp Velocity=false, Limit Acceleration=false)
  ]
  rendererClasses [&NiagaraSpriteRendererProperties]
  renderers [
   {rendererClass=&NiagaraSpriteRendererProperties}
   {rendererClass=&NiagaraSpriteRendererProperties, rendererIndex=1}
  ]
 }
 Ribbons_From_Character001 {
  emitterSpawnScript [/SetVariables(Emitter.NumberOfParticles=9) {isSetParameters=true}]
  emitterUpdateScript [
   &EmitterState(
    Life Cycle Mode=::NewEnumerator1
    Loop Duration=&MakeCustomFloatFromBool(
     Bool=&SetBoolByIntComparison(A=$Emitter.LoopCount, B=4, Comparison Type=::NewEnumerator0)
     True Float=0.1
     False Float=&RandomRangeFloat(Minimum=0.45, Maximum=0.6)
    )
    UseLoopDelay=false
   )
   &SpawnBurst_Instantaneous(Spawn Count=$Emitter.NumberOfParticles, Use Spawn Probability=false, UseLoopCountLimit=false)
   /SetVariables(
    Emitter.Random Float:NiagaraFloat=&RandomRangeFloat(Minimum=0, Maximum=1)
    Emitter.Random Picker Int:NiagaraInt32=&RandomRangeInt(Minimum=0, Maximum=7)
    Emitter.random float 2:NiagaraFloat=&RandomRangeFloat(Minimum=0, Maximum=1)
   ) {isSetParameters=true}
  ]
  particleSpawnScript [
   &PartitionParticles(
    Particles Per Partition:NiagaraInt32=$Emitter.NumberOfParticles
    Partition Count Mode:ENiagara_InfiniteLoopDuration=::NewEnumerator1
    Partition Mode:ENiagara_PartitionMode_v2=::NewEnumerator1
    Write Partition Index to Ribbon Link Order:ENiagara_AttributeSamplingApplyOutput=::NewEnumerator0
   )
   /SetVariables(
    Particles.Random Picker Float:NiagaraFloat=$Emitter.Random Float
    Particles.RandomPickerInt:NiagaraInt32=$Emitter.Random Picker Int
    Particles.RibbonID=/Niagara/DynamicInputs/NiagaraID/MakeNiagaraIDFromInt(Index:NiagaraInt32=$Emitter.LoopCount, Acquire Tag:NiagaraInt32=0)
    Particles.RibbonNumberInt:NiagaraInt32=$Emitter.LoopCount
    Particles.random picker float 2:NiagaraFloat=$Emitter.random float 2
   ) {isSetParameters=true}
   &InitializeParticle(
    Color=&MultiplyLinearColor(
     Linear Color=<rgba 1 0.853 0.6797 1>
     Scale Factor=&MakeVector4_FromLinearColor(LinearColor=$System.TODColor)
    )
    Lifetime=2.6
    Ribbon Width=&Lerp_Float(A=3, B=5, Alpha=$Particles.random picker float 2)
    Ribbon Width Mode=::NewEnumerator1
    UsePositionOffset=false
    Write Material Random=true
   )
   /SetVariables(Particles.NormalizedRibbonPos:NiagaraFloat=&OneMinusFloat(Float=&ReturnNormalizedExecIndex())) {isSetParameters=true}
   /SetVariables(Particles.SpiralStrength:NiagaraFloat=&Multiply_Float(
     A=5
     B=&FloatFromCurve(
      FloatCurve=DataInterface() {propertyValues {Curve {keys [
          {arriveTangent=3.6599, interpMode=RCIM_Cubic, leaveTangent=3.6599, tangentMode=RCTM_User, time=0, value=-0.025}
          {interpMode=RCIM_Cubic, tangentMode=RCTM_User, time=0.5, value=1}
          {arriveTangent=-0.1348, interpMode=RCIM_Cubic, leaveTangent=-0.1348, tangentMode=RCTM_User, time=1, value=0}
         ]}}}
      CurveIndex=$Particles.NormalizedRibbonPos
     )
    )) {isSetParameters=true}
   /SetVariables(Particles.RibbonU0Override=&OneMinusFloat(Float=$Particles.NormalizedRibbonPos)) {isSetParameters=true}
   /SetVariables() {isSetParameters=true}
   /SetVariables(Particles.PositionOffsetLerp:NiagaraFloat=&FloatFromCurve(
     FloatCurve=DataInterface() {propertyValues {Curve {keys [<at 0 1>,<at 1 0>]}}}
     CurveIndex=&Frac_Float(Float=&Multiply_FloatByInt(Float=0.2, Integer=$Emitter.LoopCount))
    )) {isSetParameters=true}
  ]
  particleUpdateScript [
   &ParticleState()
   &SkeletalMeshLocation(
    Bone / Socket Index=&SelectIntFromArray(
     Direct Array Index=$Particles.RibbonNumberInt
     Direct Array Mode=::NewEnumerator1
     Int Selection Array=DataInterface() {propertyValues {IntData [7,7,7]}}
    )
    Bone Sampling Mode=::NewEnumerator4
    "Inherit Velocity (Scale)":NiagaraFloat=&FloatFromCurve(
     FloatCurve=DataInterface() {propertyValues {Curve {
        defaultValue=1
        keys [{interpMode=RCIM_Cubic, tangentMode=RCTM_User, time=0.1, value=0}, {interpMode=RCIM_Cubic, tangentMode=RCTM_User, time=1, value=1}]
       }}}
     CurveIndex=$Particles.NormalizedRibbonPos
     Scale Curve=-0.7
    )
    Inherited Velocity Speed Limit=30
    MaskSpawnGroup=false
    Orientation Sampling=::NewEnumerator1
    Position Sampling=::NewEnumerator1
    Sampled Position Offset=&Lerp_Vector(
     A=&SelectVectorFromArray(
      Direct Array Index=$Particles.RibbonNumberInt
      Direct Array Mode=::NewEnumerator1
      Vector Selection Array=DataInterface() {propertyValues {FloatData [<6 0 9>, <6 0 -9>, <20 0 0>]}}
     )
     B=&Multiply_Vector(
      A=&MakeVector(
       X=&Lerp_Float(
        A=&SelectFloatFromArray(
         Direct Array Index=$Particles.RibbonNumberInt
         Direct Array Mode=::NewEnumerator1
         Float Selection Array=DataInterface() {propertyValues {FloatData [7,7,20]}}
        )
        B=&SelectFloatFromArray(
         Direct Array Index=$Particles.RibbonNumberInt
         Direct Array Mode=::NewEnumerator1
         Float Selection Array=DataInterface() {propertyValues {FloatData [7,7,25]}}
        )
        Alpha=$Particles.Random Picker Float
       )
       Y=&Lerp_Float(A=0, B=-5, Alpha=$Particles.random picker float 2)
       Z=&Lerp_Float(
        A=&SelectFloatFromArray(
         Direct Array Index=$Particles.RibbonNumberInt
         Direct Array Mode=::NewEnumerator1
         Float Selection Array=DataInterface() {propertyValues {FloatData [13,-13,8]}}
        )
        B=&SelectFloatFromArray(
         Direct Array Index=$Particles.RibbonNumberInt
         Direct Array Mode=::NewEnumerator1
         Float Selection Array=DataInterface() {propertyValues {FloatData [23,-23,-8]}}
        )
        Alpha=$Particles.Random Picker Float
       )
      )
      B=&MakeCustomVectorFromBool(Bool=$User.IsLargeBody, True Vector=<1 1 1.15>, False Vector=<1 1 1>)
     )
     Alpha=&ClampFloat(Float=&Add_Float(
       A=&FloatFromCurve(FloatCurve=DataInterface() {propertyValues {Curve {
           defaultValue=1
           keys [
            {interpMode=RCIM_Cubic, leaveTangentWeight=0.2295, tangentMode=RCTM_User, time=0.1, value=0}
            {arriveTangent=0.3797, arriveTangentWeight=0.6588, interpMode=RCIM_Cubic, leaveTangent=0.3797, tangentMode=RCTM_User, time=0.8, value=1}
           ]
          }}})
       B=&FloatFromCurve(
        FloatCurve=DataInterface() {propertyValues {Curve {
           defaultValue=1
           keys [
            {arriveTangent=1.4661, interpMode=RCIM_Cubic, leaveTangent=1.4661, leaveTangentWeight=0.3669, tangentMode=RCTM_User, time=0.0847, value=0}
            {arriveTangentWeight=0.953, interpMode=RCIM_Cubic, tangentMode=RCTM_User, time=1, value=0.7647}
           ]
          }}}
        CurveIndex=$Particles.NormalizedRibbonPos
       )
      ))
    )
    Skeletal Mesh=$System.Skeletal Mesh
    Use Sampled Mesh Attribute as Ribbon ID=false
    Velocity Sampling=::NewEnumerator0
    VelocitySpeedLimit=true
   )
   &ScaleRibbonWidth(Ribbon Width Scale=&FloatFromCurve(FloatCurve=DataInterface() {propertyValues {Curve {
        defaultValue=1
        keys [
         {arriveTangent=0.5309, interpMode=RCIM_Cubic, leaveTangent=0.5309, tangentMode=RCTM_User, time=0, value=0.5}
         {interpMode=RCIM_Cubic, tangentMode=RCTM_User, time=0.5, value=1}
         {arriveTangentWeight=0.1374, interpMode=RCIM_Cubic, leaveTangentWeight=0.07, tangentMode=RCTM_User, time=0.7, value=1}
         <at 0.9662 1.3 RCIM_Cubic>
        ]
       }}}))
   &ScaleColor(
    Scale Alpha=&FloatFromCurve(FloatCurve=DataInterface() {propertyValues {Curve {
        defaultValue=1
        keys [
         {arriveTangent=1.3708, interpMode=RCIM_Cubic, leaveTangent=1.3708, tangentMode=RCTM_User, time=0, value=0}
         {arriveTangentWeight=0.0469, interpMode=RCIM_Cubic, leaveTangentWeight=0.2197, tangentMode=RCTM_User, time=0.2729, value=1}
         {arriveTangentWeight=0.3342, interpMode=RCIM_Cubic, leaveTangentWeight=0.1014, tangentMode=RCTM_User, time=0.9, value=1}
         {arriveTangent=-5.09, interpMode=RCIM_Cubic, leaveTangent=-5.09, tangentMode=RCTM_User, time=1, value=0}
        ]
       }}})
    Scale RGB=&MakeVectorFromLinearColorRGB(Color=&ColorFromCurve(ColorCurve=DataInterface() {propertyValues {
        AlphaCurve {keys [<at 0 1 RCIM_Cubic>]}
        BlueCurve {keys [<at 0.032 2.125>,<at 0.4349 1.0156 RCIM_Cubic>,<at 1 1 RCIM_Cubic>]}
        GreenCurve {keys [<at 0.032 2.7618>,<at 0.4349 1.7965 RCIM_Cubic>,<at 1 1 RCIM_Cubic>]}
        RedCurve {keys [<at 0.032 4>,<at 0.4349 3 RCIM_Cubic>,<at 1 1 RCIM_Cubic>]}
       }}))
    ScaleA=true
    ScaleRGB=true
   )
   /SetVariables(Particles.GoalPos:NiagaraPosition=&AddVectorToPosition(
     Position=$Particles.SkeletalMeshLocation.SampledPosition
     Vector=&Multiply_Vector(
      A=&Lerp_Vector(A=<0 0 2>, B=<0 0 20>, Alpha=$Particles.NormalizedRibbonPos)
      B=&MakeVector(
       X=&Waveform(
        Global Input Position=&Add_Float(A=$Particles.random picker float 2, B=$Particles.NormalizedRibbonPos)
        "Waveform [1]"=::NewEnumerator0
        "[1] Phase Scale"=0.7
        "[1] Frequency"=1
        "[1] Amplitude Min/Max"=<-1 1>
        "[1] Scale"=5
       )
       Y=&Waveform(
        Global Input Position=&Add_Float(A=$Particles.Random Picker Float, B=$Particles.NormalizedRibbonPos)
        "Waveform [1]"=::NewEnumerator0
        "[1] Phase Scale"=0.5
        "[1] Frequency"=1
        "[1] Amplitude Min/Max"=<-1 1>
        "[1] Scale"=2
       )
       Z=&FloatFromCurve(
        FloatCurve=DataInterface() {propertyValues {Curve {
           defaultValue=1
           keys [
            {arriveTangent=1.7898, interpMode=RCIM_Cubic, leaveTangent=1.7898, leaveTangentWeight=0.0972, tangentMode=RCTM_User, time=0, value=0}
            {arriveTangent=1.5649, interpMode=RCIM_Cubic, leaveTangent=1.5649, tangentMode=RCTM_User, time=0.1651, value=0.451}
            {arriveTangent=-0.1392, interpMode=RCIM_Cubic, leaveTangent=-0.1392, tangentMode=RCTM_User, time=0.6297, value=1}
           ]
          }}}
        Scale Curve=&Lerp_Float(A=1, B=1.5, Alpha=$Particles.random picker float 2)
       )
      )
     )
    )) {isSetParameters=true}
   /SetVariables(Particles.SpiralStrength:NiagaraFloat=&Multiply_Float(
     A=&Multiply_Float(
      A=&FloatFromCurve(
       FloatCurve=DataInterface() {propertyValues {Curve {
          defaultValue=1
          keys [
           {interpMode=RCIM_Cubic, leaveTangentWeight=1.1128, tangentMode=RCTM_User, time=0.15, value=1}
           {
            arriveTangent=1.0053
            arriveTangentWeight=3.3425
            interpMode=RCIM_Cubic
            leaveTangent=1.0053
            tangentMode=RCTM_User
            time=0.8092
            value=2.6741
           }
           <at 1.0018 2.2991 RCIM_Cubic>
          ]
         }}}
       Scale Curve=&Lerp_Float(A=1.5, B=0.9, Alpha=$Particles.Random Picker Float)
      )
      B=1
     )
     B=&SelectFloatFromArray(
      Direct Array Index=$Particles.RandomPickerInt
      Direct Array Mode=::NewEnumerator1
      Float Selection Array=DataInterface() {propertyValues {FloatData [-1,1]}}
     )
    )) {isSetParameters=true}
   /SetVariables(Particles.SpiralTime:NiagaraFloat=&Add_Float(
     A=&Add_Float(
      A=&Multiply_Float(A=$Particles.Age, B=&Lerp_Float(A=0.85, B=1.2, Alpha=$Particles.random picker float 2))
      B=$Particles.NormalizedRibbonPos
     )
     B=0
    )) {isSetParameters=true}
   /SetVariables(Particles.SpiralFreq:NiagaraFloat=&Multiply_Float(
     A=&Lerp_Float(A=0.7, B=0.2, Alpha=$Particles.Random Picker Float)
     B=&FloatFromCurve(FloatCurve=DataInterface() {propertyValues {Curve {
         defaultValue=1
         keys [
          {arriveTangent=0.5251, interpMode=RCIM_Cubic, leaveTangent=0.5251, tangentMode=RCTM_User, time=0, value=0}
          {interpMode=RCIM_Cubic, tangentMode=RCTM_User, time=0.6529, value=1}
         ]
        }}})
    )) {isSetParameters=true}
   /SetVariables(Particles.SpiralOffset:Vector3f=&MakeVector(
     X=&Waveform(
      Global Input Position=$Particles.NormalizedRibbonPos
      Global Phase=$Particles.SpiralTime
      "Waveform [1]"=::NewEnumerator0
      "[1] Frequency"=$Particles.SpiralFreq
      "[1] Amplitude Min/Max"=<-1 1>
      "[1] Scale"=$Particles.SpiralStrength
     )
     Y=&Waveform(
      Global Input Position=$Particles.NormalizedRibbonPos
      Global Phase=$Particles.SpiralTime
      "Waveform [1]"=::NewEnumerator1
      "[1] Frequency"=$Particles.SpiralFreq
      "[1] Amplitude Min/Max"=<-1 2>
      "[1] Scale"=$Particles.SpiralStrength
     )
     Z=0
    )) {isSetParameters=true}
   /SetVariables(Particles.Position=&AddVectorToPosition(Position=$Particles.GoalPos, Vector=$Particles.SpiralOffset)) {isSetParameters=true}
   /SetVariables(Particles.RibbonFacing=&NormalizeVector(
     Vector To Normalize=&TransformVector(Vector=$Particles.SpiralOffset, Destination Space=::Simulation)
     Safe Normalize=true
     Safe Fallback Threshold=0
     Safe Fallback Vector=<0 0 0.001>
    )) {isSetParameters=true}
   /SetVariables(Particles.BackSpiralColor:LinearColor=&MultiplyLinearColor(
     Linear Color=$Particles.Color
     Scale Factor=&MakeVector4_FromLinearColor(LinearColor=<rgba 0.5163 0.4477 0.3644 1>)
    )) {isSetParameters=true}
   /SetVariables(Particles.BackRibbonFacing:Vector3f=&NormalizeVector(
     Vector To Normalize=&TransformVector(Vector=&Multiply_VectorByFloat(Vector=$Particles.SpiralOffset, Float=-1), Destination Space=::Simulation)
     Safe Normalize=true
     Safe Fallback Threshold=0
     Safe Fallback Vector=<0 0 0.001>
    )) {isSetParameters=true}
   &CurlNoiseForce(
    Bias Noise=true
    Bias Noise Field:Vector3f=<0 0 0.1>
    Coordinate Space:ENiagaraCoordinateSpace=::Local
    Noise Strength=5
    Pan Noise=false
    Pan Noise Field=<0 0 1>
    Write to Intrinsic Parameters=false
   )
   &Drag(Use Linear Drag=true, Use Rotational Drag=true)
   &SolveForcesAndVelocity(Clamp Velocity=false, Limit Acceleration=false)
   /SetVariables(Particles.Position=&AddVectorToPosition(
     Position=$Particles.Position
     Vector=&Add_Vector(
      A=&MakeVector(
       X=&Waveform(
        Global Input Position=&Multiply_FloatByInt(Float=0.33, Integer=$Particles.RibbonNumberInt)
        "Waveform [1]"=::NewEnumerator12
        "[1] Frequency"=0.9
        "[1] Amplitude Min/Max"=<-5 5>
        "[1] Scale"=0.7
       )
       Y=&Waveform(
        Global Input Position=&Multiply_FloatByInt(Float=0.33, Integer=$Particles.RibbonNumberInt)
        "Waveform [1]"=::NewEnumerator12
        "[1] Frequency"=1
        "[1] Amplitude Min/Max"=<-5 5>
        "[1] Scale"=0.7
       )
       Z=&FloatFromCurve(
        FloatCurve=DataInterface() {propertyValues {Curve {
           defaultValue=0
           keys [
            {arriveTangent=0.0161, interpMode=RCIM_Cubic, leaveTangent=0.0161, tangentMode=RCTM_User, time=0.3, value=0}
            {arriveTangent=1.9055, interpMode=RCIM_Cubic, leaveTangent=1.9055, tangentMode=RCTM_User, time=0.5054, value=0.2084}
            {arriveTangent=-0.5427, interpMode=RCIM_Cubic, leaveTangent=-0.5427, tangentMode=RCTM_User, time=0.9957, value=0.6075}
           ]
          }}}
        Scale Curve=&Multiply_Float(
         A=&Lerp_Float(A=25, B=30, Alpha=$Particles.random picker float 2)
         B=&MakeCustomFloatFromBool(Bool=$User.IsLargeBody, True Float=1.1, False Float=0.95)
        )
       )
      )
      B=&Multiply_VectorByFloat(
       Vector=$Output.CurlNoiseForce.CurlNoiseForce
       Float=&FloatFromCurve(FloatCurve=DataInterface() {propertyValues {Curve {
           defaultValue=1
           keys [
            {arriveTangent=0.4876, interpMode=RCIM_Cubic, leaveTangent=0.4876, leaveTangentWeight=0.1333, tangentMode=RCTM_User, time=0.75, value=0}
            {arriveTangent=5.4318, arriveTangentWeight=0.4046, interpMode=RCIM_Cubic, leaveTangent=5.4318, tangentMode=RCTM_User, time=1, value=1}
           ]
          }}})
      )
     )
    )) {isSetParameters=true}
   /SetVariables(Particles.Position=&AddVectorToPosition(
     Position=$Particles.Position
     Vector=&Multiply_Vector(
      A=&ReturnCameraVector(Camera Vectors=::NewEnumerator1, Coordinate Space=::Simulation)
      B=&VectorFromFloat(Value=6)
     )
    )) {isSetParameters=true}
   &DynamicMaterialParameters(
    Index 0 Param 1=&FloatFromCurve(
     FloatCurve=DataInterface() {propertyValues {Curve {keys [
         <at 0.3 0.2 RCIM_Cubic>
         {arriveTangent=0.5424, interpMode=RCIM_Cubic, leaveTangent=0.5424, tangentMode=RCTM_User, time=0.4946, value=1}
         {arriveTangent=-0.3824, interpMode=RCIM_Cubic, leaveTangent=-0.3824, time=0.7634, value=0.6376}
         <at 1 0.8067 RCIM_Cubic>
        ]}}}
     Scale Curve=&Lerp_Float(A=0.7, B=0.7, Alpha=$Particles.Random Picker Float)
    )
    Index 0 Param 2=$Particles.PositionOffsetLerp
    Index 0 Param 3=&FloatFromCurve(FloatCurve=DataInterface() {propertyValues {Curve {keys [
         <at 0.8 1 RCIM_Cubic>
         {arriveTangent=-0.6876, arriveTangentWeight=0.2346, interpMode=RCIM_Cubic, leaveTangent=-0.6876, tangentMode=RCTM_User, time=1, value=0.4}
        ]}}})
    Index 0 Param 4=&FloatFromCurve(FloatCurve=DataInterface() {propertyValues {Curve {
        defaultValue=1
        keys [{interpMode=RCIM_Cubic, tangentMode=RCTM_User, time=0.82, value=0.01}, {interpMode=RCIM_Cubic, tangentMode=RCTM_User, time=1, value=0.05}]
       }}})
    Module.Index0Float0Write=true
    Module.Index0Float1Write=true
    Module.Index0Float2Write=true
    Module.Index0Float3Write=true
   )
   /SetVariables(
    Particles.Glow Ribbon Color:LinearColor=&MultiplyLinearColor(
     Linear Color=&ColorFromCurve(ColorCurve=DataInterface() {propertyValues {
        AlphaCurve {keys [<at 0 0>,<at 0.0946 1>,<at 0.3149 1 RCIM_Cubic>,<at 1 0 RCIM_Cubic>]}
        BlueCurve {keys [<at 0.0049 0.1771>, {arriveTangent=0.09, interpMode=RCIM_Cubic, leaveTangent=0.09, time=0.2281, value=0.3906}, <at 0.8153 0.25 RCIM_Cubic>]}
        GreenCurve {keys [
          <at 0.0049 0.5282>
          {arriveTangent=0.0884, interpMode=RCIM_Cubic, leaveTangent=0.0884, time=0.2281, value=1.0773}
          <at 0.8153 0.5999 RCIM_Cubic>
         ]}
        RedCurve {keys [<at 0.0049 1>,<at 0.2281 2 RCIM_Cubic>,<at 0.8153 1 RCIM_Cubic>]}
       }})
     Scale Factor=&MakeVector4(
      X=1
      Y=1
      Z=1
      W=&MakeCustomFloatFromBool(Bool=$User.bIsFrontEndPreview, True Float=0.5, False Float=1)
     )
    )
    Particles.Glow Ribbon Width:NiagaraFloat=&Multiply_Float(
     A=$Particles.RibbonWidth
     B=&FloatFromCurve(
      FloatCurve=DataInterface() {propertyValues {Curve {keys [
          <at 0.0039 0.4 RCIM_Cubic>
          {arriveTangentWeight=0.036, interpMode=RCIM_Cubic, leaveTangentWeight=0.1554, tangentMode=RCTM_User, time=0.25, value=1}
          {arriveTangentWeight=0.1325, interpMode=RCIM_Cubic, leaveTangentWeight=0.1338, tangentMode=RCTM_User, time=0.4933, value=1}
          {arriveTangentWeight=0.1552, interpMode=RCIM_Cubic, leaveTangentWeight=0.0422, tangentMode=RCTM_User, time=0.75, value=1}
          <at 1.0039 0.4 RCIM_Cubic>
         ]}}}
      CurveIndex=$Particles.NormalizedRibbonPos
      Scale Curve=&FloatFromCurve(FloatCurve=DataInterface() {propertyValues {Curve {keys [
           {arriveTangent=18.2755, interpMode=RCIM_Cubic, leaveTangent=18.2755, tangentMode=RCTM_User, time=0, value=1.5}
           {arriveTangentWeight=0.4142, interpMode=RCIM_Cubic, leaveTangentWeight=0.4439, tangentMode=RCTM_User, time=0.1818, value=2.5}
           {arriveTangentWeight=0.396, interpMode=RCIM_Cubic, leaveTangentWeight=0.4837, tangentMode=RCTM_User, time=0.5, value=2.5}
           {arriveTangentWeight=0.8967, interpMode=RCIM_Cubic, tangentMode=RCTM_User, time=1, value=1}
          ]}}})
     )
    )
   ) {isSetParameters=true}
  ]
  rendererClasses [&NiagaraRibbonRendererProperties]
  renderers [
   {rendererClass=&NiagaraRibbonRendererProperties}
   {rendererClass=&NiagaraRibbonRendererProperties, rendererIndex=1}
   {rendererClass=&NiagaraRibbonRendererProperties, rendererIndex=2}
   {rendererClass=&NiagaraRibbonRendererProperties, rendererIndex=3}
  ]
 }
 mesh_curls {
  bEnabled=false
  emitterUpdateScript [
   &EmitterState()
   &SpawnBurst_Instantaneous(Spawn Count=2, Use Spawn Probability=false, UseLoopCountLimit=true)
   &SpawnRate(
    SpawnRate=&RandomRangeFloat(Minimum=1.75, Maximum=2, Evaluation Type=::NewEnumerator1)
    Use Spawn Probability=true
   )
  ]
  particleSpawnScript [
   &InitializeParticle(
    Color=&MultiplyLinearColor(
     Linear Color=<rgba 1 0.853 0.6797 1>
     Scale Factor=&MakeVector4_FromLinearColor(LinearColor=$System.TODColor)
    )
    Lifetime=&RandomRangeFloat(Minimum=1.3, Maximum=1.65)
    Mesh Scale=&Multiply_VectorByFloat(
     Vector=<0.1 0.1 0.1>
     Float=&Multiply_Float(
      A=&RandomRangeFloat(Minimum=0.7, Maximum=0.85)
      B=&MakeCustomFloatFromBool(Bool=$User.IsLargeBody, True Float=1.1, False Float=1)
     )
    )
    Mesh Scale Mode=::NewEnumerator4
    Sprite Rotation Mode=::NewEnumerator2
    Sprite Size Mode=::NewEnumerator3
    Sprite UV Mode=::NewEnumerator3
    Uniform Sprite Size=&RandomRangeFloat(Minimum=4, Maximum=6)
    UsePositionOffset=false
    Write Material Random=true
   )
   &PartitionParticles()
   &AddVelocity(
    Velocity=&MergeHelper_VectorSwitch(0=<0.52 0 1>, 1=<-0.52 0 1>, 2=?, 3=?, Index=$Particles.PartitionParticles.Partition, Num=::NewEnumerator0)
    Velocity Speed Scale=&Multiply_Float(
     A=&RandomRangeFloat(Minimum=18, Maximum=25)
     B=&MakeCustomFloatFromBool(Bool=$User.IsLargeBody, True Float=1.15, False Float=1)
    )
   )
   &InitialMeshOrientation(
    Mesh Orientation Mode=::NewEnumerator2
    Orientation Vector=<0 0 1>
    Rotation=&RandomRangeVector(Minimum=<0 0 0>, Maximum=<0 0 1>)
    Use Rotation Vector=true
   )
   &SkeletalMeshLocation(
    Bone / Socket Index=8
    Bone Sampling Mode=::NewEnumerator4
    MaskSpawnGroup=false
    Position Sampling=::NewEnumerator2
    Skeletal Mesh=$System.Skeletal Mesh
   ) {enabled=false, module=SkeletalMeshLocation001}
  ]
  particleUpdateScript [
   &ParticleState()
   &UpdateMeshOrientation(
    Rotation Rate=&Multiply_Float(A=&RandomRangeFloat(Minimum=0.3, Maximum=0.8), B=-1)
    Rotation Vector=&MakeVector(X=0, Y=0, Z=1)
   )
   &SkeletalMeshLocation(
    Bone / Socket Index=&SelectIntFromArray(
     Direct Array Index=$Particles.PartitionParticles.Partition
     Direct Array Mode=::NewEnumerator1
     Int Selection Array=DataInterface() {propertyValues {IntData [1,2]}}
    )
    Bone Sampling Mode=::NewEnumerator4
    MaskSpawnGroup=false
    Orientation Sampling=::NewEnumerator1
    Position Sampling=::NewEnumerator1
    Sampled Position Offset=&MergeHelper_VectorSwitch(
     0=&Lerp_Vector(A=<7 0 0>, B=<15 0 0>, Alpha=&RandomRangeFloat(Minimum=0, Maximum=1))
     1=&Lerp_Vector(A=<-7 0 0>, B=<-15 0 0>, Alpha=&RandomRangeFloat(Minimum=0, Maximum=1))
     2=?
     3=?
     Index=$Particles.PartitionParticles.Partition
     Num=::NewEnumerator0
    )
    Skeletal Mesh=$System.Skeletal Mesh
    Use Sampled Mesh Attribute as Ribbon ID=false
   )
   &ScaleMeshSize(Scale Factor=&Multiply_VectorByFloat(
     Vector=&MakeVector(
      X=&FloatFromCurve(FloatCurve=DataInterface() {propertyValues {Curve {keys [
           <at 0 1.2 RCIM_Cubic>
           {arriveTangent=-0.1903, interpMode=RCIM_Cubic, leaveTangent=-0.1903, time=0.8, value=0.929}
           {arriveTangent=0.6374, interpMode=RCIM_Cubic, leaveTangent=0.6374, tangentMode=RCTM_User, time=0.998, value=1.0101}
          ]}}})
      Y=&FloatFromCurve(FloatCurve=DataInterface() {propertyValues {Curve {keys [
           <at 0 1.2 RCIM_Cubic>
           {arriveTangent=-0.2178, interpMode=RCIM_Cubic, leaveTangent=-0.2178, time=0.8, value=0.9269}
           {arriveTangent=0.3322, interpMode=RCIM_Cubic, leaveTangent=0.3322, tangentMode=RCTM_User, time=0.9991, value=0.9824}
          ]}}})
      Z=&FloatFromCurve(
       FloatCurve=DataInterface() {propertyValues {Curve {keys [
           {arriveTangent=3.0747, interpMode=RCIM_Cubic, leaveTangent=3.0747, tangentMode=RCTM_User, time=0, value=0.6}
           {arriveTangent=0.5333, interpMode=RCIM_Cubic, leaveTangent=0.5333, time=0.15, value=0.85}
           {arriveTangent=0.4243, interpMode=RCIM_Cubic, leaveTangent=0.4243, time=0.75, value=1}
           {arriveTangent=0.9676, interpMode=RCIM_Cubic, leaveTangent=0.9676, tangentMode=RCTM_User, time=1, value=1.2106}
          ]}}}
       Scale Curve=&RandomRangeFloat(Minimum=0.9, Maximum=1.2)
      )
     )
     Float=&FloatFromCurve(FloatCurve=DataInterface() {propertyValues {Curve {
         defaultValue=1
         keys [
          {arriveTangent=3.7509, interpMode=RCIM_Cubic, leaveTangent=3.7509, tangentMode=RCTM_User, time=0, value=0.5}
          {arriveTangentWeight=0.1274, interpMode=RCIM_Cubic, leaveTangentWeight=0.7778, tangentMode=RCTM_User, time=0.2759, value=0.9}
          {arriveTangent=-0.0405, interpMode=RCIM_Cubic, leaveTangent=-0.0405, tangentMode=RCTM_User, time=1, value=1}
         ]
        }}})
    ))
   &DynamicMaterialParameters(
    Index 0 Param 1=&FloatFromCurve(
     FloatCurve=DataInterface() {propertyValues {Curve {keys [
         <at 0.3 0.2 RCIM_Cubic>
         {arriveTangent=0.5424, interpMode=RCIM_Cubic, leaveTangent=0.5424, tangentMode=RCTM_User, time=0.605, value=0.6745}
         <at 1.0013 0.7537 RCIM_Cubic>
        ]}}}
     Scale Curve=&Lerp_Float(A=0.7, B=0.8, Alpha=&RandomRangeFloat(Minimum=0, Maximum=1))
    )
    Index 0 Param 2=&RandomRangeFloat(Minimum=0, Maximum=1)
    Index 0 Param 3=&FloatFromCurve(FloatCurve=DataInterface() {propertyValues {Curve {keys [
         <at 0.6 1 RCIM_Cubic>
         {arriveTangent=-0.6876, arriveTangentWeight=0.2346, interpMode=RCIM_Cubic, leaveTangent=-0.6876, tangentMode=RCTM_User, time=1, value=0.4}
        ]}}})
    Index 0 Param 4=&FloatFromCurve(FloatCurve=DataInterface() {propertyValues {Curve {
        defaultValue=1
        keys [{interpMode=RCIM_Cubic, tangentMode=RCTM_User, time=0.6, value=0.01}, {interpMode=RCIM_Cubic, tangentMode=RCTM_User, time=1, value=0.05}]
       }}})
    Module.Index0Float0Write=true
    Module.Index0Float1Write=true
    Module.Index0Float2Write=true
    Module.Index0Float3Write=true
   )
   &Drag(Drag=&RandomRangeFloat(Minimum=0.8, Maximum=1.1), Use Linear Drag=true, Use Rotational Drag=true)
   &ScaleColor(
    Scale Alpha=&FloatFromCurve(FloatCurve=DataInterface() {propertyValues {Curve {
        defaultValue=1
        keys [
         {arriveTangent=1.3708, interpMode=RCIM_Cubic, leaveTangent=1.3708, tangentMode=RCTM_User, time=0, value=0}
         {arriveTangentWeight=0.0469, interpMode=RCIM_Cubic, leaveTangentWeight=0.2197, tangentMode=RCTM_User, time=0.1, value=1}
         {arriveTangentWeight=0.3342, interpMode=RCIM_Cubic, leaveTangentWeight=0.1014, tangentMode=RCTM_User, time=0.6047, value=1}
         {arriveTangent=-1.7747, interpMode=RCIM_Cubic, leaveTangent=-1.7747, tangentMode=RCTM_User, time=1, value=0}
        ]
       }}})
    Scale RGB=&MakeVectorFromLinearColorRGB(Color=&ColorFromCurve(ColorCurve=DataInterface() {propertyValues {
        AlphaCurve {keys [<at 0 1 RCIM_Cubic>]}
        BlueCurve {keys [<at 0 2.125>, {arriveTangent=-1.125, interpMode=RCIM_Cubic, leaveTangent=-1.125, time=0.1778, value=1.0156}, <at 1 1 RCIM_Cubic>]}
        GreenCurve {keys [<at 0 2.7618>, {arriveTangent=-1.7618, interpMode=RCIM_Cubic, leaveTangent=-1.7618, time=0.1778, value=1.7965}, <at 1 1 RCIM_Cubic>]}
        RedCurve {keys [<at 0 4>, {arriveTangent=-3, interpMode=RCIM_Cubic, leaveTangent=-3, time=0.1778, value=3}, <at 1 1 RCIM_Cubic>]}
       }}))
    ScaleA=true
    ScaleRGB=true
   )
   &SolveForcesAndVelocity(Clamp Velocity=false, Limit Acceleration=false)
   &CameraOffset(Camera Offset Amount=-10)
  ]
  rendererClasses [&NiagaraMeshRendererProperties]
  renderers [{rendererClass=&NiagaraMeshRendererProperties, rendererIndex=0}]
 }
}
summary {
 systemName=NS_Aura_FilmSparkle
 emitters [
  {bEnabled=false, emitterName=Motes_Flares002, rendererClasses [&NiagaraSpriteRendererProperties]}
  {emitterName=Ribbons_From_Character001, rendererClasses [&NiagaraRibbonRendererProperties]}
  {emitterName=RibbonParticles002, rendererClasses [&NiagaraSpriteRendererProperties]}
  {emitterName=RibbonSparkles002, rendererClasses [&NiagaraSpriteRendererProperties]}
  {bEnabled=false, emitterName=mesh_curls, rendererClasses [&NiagaraMeshRendererProperties]}
  {emitterName=Motes_From_Character, rendererClasses [&NiagaraSpriteRendererProperties]}
  {emitterName=Decal_GlowRays, rendererClasses [&NiagaraSpriteRendererProperties]}
  {
   emitterName=Motes_Flares_MeshRibbons
   rendererClasses [&NiagaraSpriteRendererProperties, &NiagaraMeshRendererProperties]
  }
 ]
 userVariables [
  User.bIsFrontEnd {type=NiagaraBool}
  User.bIsFrontEndPreview {type=NiagaraBool}
  User.IsLargeBody {type=NiagaraBool}
 ]
}
