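# Niagara start burst for the FilmSparkle aura cosmetic: a brief warm flare, corona glow and flickering motes on a character's skeletal mesh.
# Flare and Corona_Glow_Combined follow the mesh, with curve-driven fades, sizes and time-of-day alpha.
# Motes_Flares001 launches sprite and mesh particles outward from the mesh, using curl noise, drag and a flicker waveform.
# MergeHelper switches keyed on spawn group pack several layers into single emitters.
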NiagaraSystem /SparksCosmetics/Items/Auras/Assets/NS_Aura_FilmSparkle/NS_Aura_FilmSparkle_Start
dependencies {
 usedDataInterfaces [&NiagaraDataInterfaceCurve, &NiagaraDataInterfaceColorCurve]
 usedDynamicInputs []
 usedModules [
  &self:SystemSpawnScript.SystemScriptSource.SystemScriptGraph.NiagaraNodeAssignment_0.SetVariables_#1
  &SystemState
  &self:SystemSpawnScript.SystemScriptSource.SystemScriptGraph.NiagaraNodeAssignment_1.SetVariables_#2
 ]
}
emitters {
 Corona_Glow_Combined {
  emitterUpdateScript [
   &EmitterState(Life Cycle Mode=::NewEnumerator1, Loop Behavior=::NewEnumerator1, Loop Duration=1, UseLoopDelay=false)
   &SpawnBurst_Instantaneous(Spawn Count=1, Use Spawn Probability=false)
   &SpawnBurst_Instantaneous(Spawn Count=1, Spawn Group=1, Use Spawn Probability=false) {module=SpawnBurst_Instantaneous001}
  ]
  particleSpawnScript [
   /SetVariables(Particles.VisibilityTag=&SpawnGroup()) {isSetParameters=true}
   &InitializeParticle(
    Color=&MergeHelper_ColorSwitch(0=<rgba 1 0.628 0.2718 1>, 1=<rgba 0.1051 0.0587 0.0321 1>, 2=?, 3=?, Index=&SpawnGroup(), Num=::NewEnumerator0)
    Lifetime=&MergeHelper_FloatSwitch(0=0.3, 1=0.4, 2=?, 3=?, Index=&SpawnGroup(), Num=::NewEnumerator0)
    Sprite Rotation Mode=::NewEnumerator2
    Sprite Size Mode=::NewEnumerator3
    Uniform Sprite Size=&MergeHelper_FloatSwitch(0=400, 1=230, 2=?, 3=?, Index=&SpawnGroup(), Num=::NewEnumerator0)
    UsePositionOffset=false
    Write Material Random=true
   )
   &CameraOffset(Camera Offset Amount=&MergeHelper_FloatSwitch(0=-30, 1=-10, 2=?, 3=?, Index=&SpawnGroup(), Num=::NewEnumerator0))
   &SystemLocation(
    MaskSpawnGroup=false
    Offset=&MergeHelper_VectorSwitch(0=<0 0 120>, 1=<0 0 0>, 2=?, 3=?, Index=&SpawnGroup(), Num=::NewEnumerator0)
   )
  ]
  particleUpdateScript [
   &ParticleState()
   &SkeletalMeshLocation(Bone Sampling Mode=::NewEnumerator4, MaskSpawnGroup=false, Skeletal Mesh=$System.Skeletal Mesh) {module=SkeletalMeshLocation001}
   &ScaleColor(
    Scale Alpha=&MergeHelper_FloatSwitch(
     0=&Multiply_Float(
      A=&FloatFromCurve(
       FloatCurve=DataInterface() {propertyValues {Curve {keys [
           <at 0 0.85 RCIM_Cubic>
           {arriveTangent=0.3, interpMode=RCIM_Cubic, leaveTangent=0.3, time=0.2009, value=0.92}
           <at 0.5 1 RCIM_Cubic>
           {arriveTangent=-0.3, interpMode=RCIM_Cubic, leaveTangent=-0.3, time=0.8028, value=0.92}
           <at 1 0.85 RCIM_Cubic>
          ]}}}
       CurveIndex=Linked() {linkedVariable {name=NPC.FortniteNPC.TimeOfDay_0-1, type=NiagaraFloat}}
      )
      B=&FloatFromCurve(FloatCurve=DataInterface() {propertyValues {Curve {keys [
           {arriveTangent=9.6062, interpMode=RCIM_Cubic, leaveTangent=9.6062, tangentMode=RCTM_User, time=0, value=0}
           {interpMode=RCIM_Cubic, tangentMode=RCTM_User, time=0.2, value=1}
           {interpMode=RCIM_Cubic, tangentMode=RCTM_User, time=1, value=0}
          ]}}})
     )
     1=&FloatFromCurve(
      FloatCurve=DataInterface() {propertyValues {Curve {keys [
          {interpMode=RCIM_Cubic, tangentMode=RCTM_User, time=0, value=0}
          {interpMode=RCIM_Cubic, tangentMode=RCTM_User, time=0.2, value=1}
          {interpMode=RCIM_Cubic, tangentMode=RCTM_User, time=0.5, value=1}
          {tangentMode=RCTM_User, time=1, value=0}
         ]}}}
      Scale Curve=2
     )
     2=?
     3=?
     Index=&SpawnGroup()
     Num=::NewEnumerator0
    )
    Scale RGB=&MergeHelper_VectorSwitch(
     0=&Multiply_VectorByFloat(
      Vector=<1 1 1>
      Float=&FloatFromCurve(FloatCurve=DataInterface() {propertyValues {Curve {keys [{arriveTangent=-2.9031, interpMode=RCIM_Cubic, leaveTangent=-2.9031, tangentMode=RCTM_User, time=0, value=2}, <at 1 1 RCIM_Cubic>]}}})
     )
     1=&MakeVectorFromLinearColorRGB(Color=&ColorFromCurve(ColorCurve=DataInterface() {propertyValues {
         AlphaCurve {keys [<at 0 1 RCIM_Cubic>]}
         BlueCurve {keys [<at 0.1905 1 RCIM_Cubic>,<at 1 0.4974 RCIM_Cubic>]}
         GreenCurve {keys [<at 0.1905 1 RCIM_Cubic>,<at 1 0 RCIM_Cubic>]}
         RedCurve {keys [<at 0.1905 1 RCIM_Cubic>,<at 1 0.3623 RCIM_Cubic>]}
        }}))
     2=?
     3=?
     Index=&SpawnGroup()
     Num=::NewEnumerator0
    )
    ScaleA=true
    ScaleRGB=true
   )
   &ScaleSpriteSize(Scale Factor=&MergeHelper_Vector2Switch(
     0=&Vector2DFromFloat(Value=&FloatFromCurve(FloatCurve=DataInterface() {propertyValues {Curve {keys [
           {arriveTangent=2.1722, interpMode=RCIM_Cubic, leaveTangent=2.1722, tangentMode=RCTM_User, time=0, value=0}
           {arriveTangent=0.0224, interpMode=RCIM_Cubic, leaveTangent=0.0224, tangentMode=RCTM_User, time=1, value=1}
          ]}}}))
     1=<1 1>
     2=?
     3=?
     4=?
     Index=&SpawnGroup()
     Num=::NewEnumerator0
    ))
   &DynamicMaterialParameters(
    Index 0 Param 1=0.75
    Index 0 Param 2=2
    Index 0 Param 3=1
    Module.Index0Float0Write=true
    Module.Index0Float1Write=true
    Module.Index0Float2Write=true
    Module.Index0Float3Write=true
   ) {module=DynamicMaterialParameters001}
   /SetVariables(Particles.GlowDynamicParameter:Vector4f=<3 1 1 1>) {isSetParameters=true}
  ]
  rendererClasses [&NiagaraSpriteRendererProperties]
  renderers [
   {rendererClass=&NiagaraSpriteRendererProperties}
   {rendererClass=&NiagaraSpriteRendererProperties, rendererIndex=1}
  ]
 }
 Flare {
  emitterUpdateScript [
   &EmitterState(Life Cycle Mode=::NewEnumerator1, Loop Behavior=::NewEnumerator1, Loop Duration=1, UseLoopDelay=false)
   &SpawnBurst_Instantaneous(Spawn Count=1, Use Spawn Probability=false)
  ]
  particleSpawnScript [
   &InitializeParticle(
    Color=<rgba 1.25 0.5635 0.1774 1>
    Lifetime=0.3
    Sprite Rotation Mode=::NewEnumerator2
    Sprite Size Mode=::NewEnumerator3
    Uniform Sprite Size=400
    UsePositionOffset=false
    Write Material Random=true
   )
   &DynamicMaterialParameters(
    Index 0 Param 1=&RandomRangeFloat(Minimum=0.4, Maximum=0.6)
    Index 0 Param 2=1
    Index 0 Param 3=1
    Index 0 Param 4=1
    Module.Param0WriteEnabled=true
    Module.Param1WriteEnabled=true
    Module.Param2WriteEnabled=true
    Module.Param3WriteEnabled=true
   )
   &SystemLocation(MaskSpawnGroup=false, Offset=<0 0 120>)
   /SetVariables(Particles.Sharp Flare Camera Offset=40) {isSetParameters=true}
   &CameraOffset(Camera Offset Amount=-30) {module=CameraOffset001}
  ]
  particleUpdateScript [
   &ParticleState()
   &SolveForcesAndVelocity(Clamp Velocity=false, Limit Acceleration=false)
   &SkeletalMeshLocation(Bone Sampling Mode=::NewEnumerator4, MaskSpawnGroup=false, Skeletal Mesh=$System.Skeletal Mesh) {module=SkeletalMeshLocation001}
   &ScaleColor(
    Scale Alpha=&Multiply_Float(
     A=&FNTimeOfDayScalar(Night Scale=1, Morning Scale=1, Evening Scale=1)
     B=&FloatFromCurve(FloatCurve=DataInterface() {propertyValues {Curve {keys [
          {arriveTangent=8.1746, interpMode=RCIM_Cubic, leaveTangent=8.1746, tangentMode=RCTM_User, time=0, value=0}
          {arriveTangentWeight=0.1067, interpMode=RCIM_Cubic, leaveTangentWeight=0.3202, tangentMode=RCTM_User, time=0.2, value=1}
          {arriveTangentWeight=0.3202, interpMode=RCIM_Cubic, leaveTangentWeight=0.1067, tangentMode=RCTM_User, time=0.5, value=1}
          <at 1 0 RCIM_Cubic>
         ]}}})
    )
    Scale RGB=&Multiply_VectorByFloat(Vector=<1 1 1>, Float=1)
    ScaleA=true
    ScaleRGB=true
   )
   &ScaleSpriteSize(Scale Factor=&Vector2DFromFloat(Value=&FloatFromCurve(FloatCurve=DataInterface() {propertyValues {Curve {keys [<at 0 1.5 RCIM_Cubic>,<at 1 0>]}}})))
   /SetVariables(Particles.Flash color:LinearColor=&MakeLinearColorFromVectorAndFloat(
     "Vector (RGB)"=&MakeVectorFromLinearColorRGB(Color=<rgba 1 0.655 0.3251 1>)
     "Float (Alpha)"=&Multiply_Float(
      A=&Multiply_Float(
       A=&FNTimeOfDayScalar(Night Scale=1, Morning Scale=1, Evening Scale=1)
       B=&FloatFromCurve(FloatCurve=DataInterface() {propertyValues {Curve {keys [
            {arriveTangent=14.8591, interpMode=RCIM_Cubic, leaveTangent=14.8591, tangentMode=RCTM_User, time=0, value=0}
            {arriveTangentWeight=0.093, interpMode=RCIM_Cubic, leaveTangentWeight=0.2791, tangentMode=RCTM_User, time=0.1, value=1}
            {arriveTangentWeight=0.2791, interpMode=RCIM_Cubic, leaveTangentWeight=0.093, tangentMode=RCTM_User, time=0.5, value=1}
            <at 0.7 0 RCIM_Cubic>
           ]}}})
      )
      B=1
     )
    )) {isSetParameters=true}
   /SetVariables(Particles.Flash size:Vector2f=&Multiply_Vector2D_ByFloat(
     Float=&Multiply_Float(
      A=375
      B=&FloatFromCurve(FloatCurve=DataInterface() {propertyValues {Curve {keys [
           {arriveTangent=16.6356, interpMode=RCIM_Cubic, leaveTangent=16.6356, tangentMode=RCTM_User, time=0, value=0}
           {arriveTangent=-0.15, interpMode=RCIM_Cubic, leaveTangent=-0.15, tangentMode=RCTM_User, time=0.1378, value=1}
           {arriveTangent=-0.0405, interpMode=RCIM_Cubic, leaveTangent=-0.0405, tangentMode=RCTM_User, time=0.62, value=0.25}
          ]}}})
     )
     Vector2D=<1 1>
    )) {isSetParameters=true}
  ]
  rendererClasses [&NiagaraSpriteRendererProperties]
  renderers [
   {rendererClass=&NiagaraSpriteRendererProperties}
   {rendererClass=&NiagaraSpriteRendererProperties, rendererIndex=1}
  ]
 }
 Motes_Flares001 {
  emitterUpdateScript [
   &EmitterState(Life Cycle Mode=::NewEnumerator1, Loop Behavior=::NewEnumerator1, Loop Delay=1.15, Loop Duration=0.15, UseLoopDelay=false)
   &SpawnBurst_Instantaneous(Spawn Count=5, Use Spawn Probability=false, UseLoopCountLimit=true)
   &SpawnRate(SpawnRate=200, Use Spawn Probability=true)
   &SpawnBurst_Instantaneous(Spawn Count=12, Spawn Group=1, Use Spawn Probability=false, UseLoopCountLimit=true) {module=SpawnBurst_Instantaneous001}
   &SpawnBurst_Instantaneous(Spawn Count=6, Spawn Group=2, Use Spawn Probability=false, UseLoopCountLimit=false) {enabled=false, module=SpawnBurst_Instantaneous002}
  ]
  particleSpawnScript [
   /SetVariables(Particles.VisibilityTag=&SpawnGroup()) {isSetParameters=true}
   &InitializeParticle(
    Color=&RandomRangeLinearColor(Minimum=<rgba 1 0.4865 0.2632 1>, Maximum=<rgba 1 0.8182 0.0938 1>)
    Lifetime=&MergeHelper_FloatSwitch(
     0=&RandomRangeFloat(Minimum=1.2, Maximum=1.5)
     1=&RandomRangeFloat(Minimum=0.8, Maximum=1.5)
     2=&RandomRangeFloat(Minimum=0.6, Maximum=0.9)
     3=?
     Index=&SpawnGroup()
     Num=::NewEnumerator1
    )
    Mesh Scale=&Multiply_VectorByFloat(
     Vector=<0.1 0.1 0.1>
     Float=&Multiply_Float(
      A=&RandomRangeFloat(Minimum=0.7, Maximum=0.85)
      B=&MakeCustomFloatFromBool(Bool=$User.IsLargeBody, True Float=1.1, False Float=1)
     )
    )
    Mesh Scale Mode=::NewEnumerator4
    Sprite Size Mode=::NewEnumerator3
    Uniform Sprite Size=&MergeHelper_FloatSwitch(
     0=&RandomRangeFloat(Minimum=1, Maximum=2)
     1=&RandomRangeFloat(Minimum=9, Maximum=15)
     2=?
     3=?
     Index=&SpawnGroup()
     Num=::NewEnumerator0
    )
    UsePositionOffset=false
    Write Material Random=true
   )
   /SetVariables(Particles.RandomCore:NiagaraFloat=&FloatFromCurve(
     FloatCurve=DataInterface() {propertyValues {Curve {keys [
         {arriveTangent=0.0559, interpMode=RCIM_Cubic, leaveTangent=0.0559, tangentMode=RCTM_User, time=0, value=0.2}
         {arriveTangent=0.1895, interpMode=RCIM_Cubic, leaveTangent=0.1895, tangentMode=RCTM_User, time=0.9, value=0.25}
         {arriveTangent=1.1808, interpMode=RCIM_Cubic, leaveTangent=1.1808, tangentMode=RCTM_User, time=1, value=0.3}
        ]}}}
     CurveIndex=&RandomRangeFloat(Minimum=0, Maximum=1)
     Scale Curve=2
    )) {isSetParameters=true}
   &ShapeLocation(
    Hemisphere Distribution=&MergeHelper_Vector2Switch(0=<1 1>, 1=<1 0.7>, 2=<1 0.5>, 3=?, 4=?, Index=&SpawnGroup(), Num=::NewEnumerator1)
    MaskSpawnGroup=false
    Override Seed=false
    Sphere Radius=&MergeHelper_FloatSwitch(0=10, 1=20, 2=15, 3=?, Index=&SpawnGroup(), Num=::NewEnumerator1)
    Sphere Surface Distribution=0.5405
    UseOwnerScale=true
   )
   &AddVelocity(
    Module.Use Falloff Distance=false
    Origin Offset=&MergeHelper_VectorSwitch(0=<0 0 0>, 1=<0 0 -5>, 2=<0 0 -7>, 3=?, Index=&SpawnGroup(), Num=::NewEnumerator1)
    Override Seed=false
    Velocity Mode=::NewEnumerator1
    Velocity Speed=&MergeHelper_FloatSwitch(
     0=&RandomRangeFloat(Minimum=175, Maximum=375)
     1=&RandomRangeFloat(Minimum=125, Maximum=250)
     2=&RandomRangeFloat(Minimum=150, Maximum=200)
     3=?
     Index=&SpawnGroup()
     Num=::NewEnumerator1
    )
   )
   &CameraOffset(Camera Offset Amount=-25)
   &SkeletalMeshLocation(
    Bone Sampling Mode=::NewEnumerator4
    MaskSpawnGroup=false
    Orientation Sampling=::NewEnumerator1
    Position Sampling=::NewEnumerator1
    Sampled Position Offset=&AddVectorToPosition(Position=$Particles.ShapeLocation.ShapePosition, Vector=<15 0 0>)
    Skeletal Mesh=$System.Skeletal Mesh
   )
   &InitialMeshOrientation(
    Mesh Orientation Mode=::NewEnumerator2
    Orientation Vector=<0 0 1>
    Rotation=&RandomRangeVector(Minimum=<0 0 0>, Maximum=<0 0 1>)
    Use Rotation Vector=true
   )
  ]
  particleUpdateScript [
   &ParticleState()
   /SetVariables(Particles.FlickerWaveform:NiagaraFloat=&Waveform(
     Global Phase=&Add_Float(A=$Engine.Time, B=&RandomRangeFloat(Minimum=0, Maximum=1))
     "Waveform [1]"=::NewEnumerator10
     "[1] Frequency"=15
     "[1] Amplitude Min/Max"=<0.2 1.2>
    )) {isSetParameters=true}
   &CurlNoiseForce(
    Noise Strength=&Multiply_Float(
     A=&FloatFromCurve(
      FloatCurve=DataInterface() {propertyValues {Curve {
         defaultValue=1
         keys [
          {interpMode=RCIM_Cubic, leaveTangentWeight=0.5765, tangentMode=RCTM_User, time=0, value=0}
          {interpMode=RCIM_Cubic, tangentMode=RCTM_User, time=0.9, value=1}
         ]
        }}}
      Scale Curve=&MergeHelper_FloatSwitch(0=150, 1=100, 2=10, 3=?, Index=&SpawnGroup(), Num=::NewEnumerator1)
     )
     B=1
    )
    Pan Noise=true
    Pan Noise Field=<0 0 1>
   )
   &UpdateMeshOrientation(
    Rotation Rate=&Multiply_Float(A=&RandomRangeFloat(Minimum=0.4, Maximum=0.85), B=1)
    Rotation Vector=&MakeVector(X=0, Y=0, Z=1)
   )
   &ScaleSpriteSize(Scale Factor=&MergeHelper_Vector2Switch(
     0=&Vector2DFromFloat(Value=&FloatFromCurve(
       FloatCurve=DataInterface() {propertyValues {Curve {keys [
           {arriveTangent=7.4528, interpMode=RCIM_Cubic, leaveTangent=7.4528, tangentMode=RCTM_User, time=0, value=0}
           <at 0.2 1.2 RCIM_Cubic>
           {arriveTangent=-2.7132, interpMode=RCIM_Cubic, leaveTangent=-2.7132, tangentMode=RCTM_User, time=1, value=0}
          ]}}}
       Scale Curve=&Multiply_Float(A=&FN_FrontEndFloatModifier(Gameplay=0.6, Front End=0.5), B=$Particles.FlickerWaveform)
      ))
     1=&Multiply_Vector2D_ByFloat(
      Float=&Multiply_Float(
       A=&FloatFromCurve(FloatCurve=DataInterface() {propertyValues {Curve {keys [
            {arriveTangent=7.4528, interpMode=RCIM_Cubic, leaveTangent=7.4528, tangentMode=RCTM_User, time=0, value=0}
            <at 0.2 1.2 RCIM_Cubic>
            {arriveTangent=-2.7132, interpMode=RCIM_Cubic, leaveTangent=-2.7132, tangentMode=RCTM_User, time=1, value=0}
           ]}}})
       B=/Game/Effects/Niagara/DynamicInputs/Math/FN_Fit(
        Clamp:NiagaraBool=false
        Value:NiagaraFloat=$Particles.FlickerWaveform
        SourceMin:NiagaraFloat=0.2
        SourceMax:NiagaraFloat=1.2
        DestMin:NiagaraFloat=0.5
        DestMax:NiagaraFloat=1
       )
      )
      Vector2D=&MakeVector2D(
       X=&FloatFromCurve(FloatCurve=DataInterface() {propertyValues {Curve {keys [<at 0 0.8>,<at 1 1>]}}})
       Y=&FloatFromCurve(FloatCurve=DataInterface() {propertyValues {Curve {keys [<at 0 1.2529>,<at 1 1>]}}})
      )
     )
     2=?
     3=?
     4=?
     Index=&SpawnGroup()
     Num=::NewEnumerator0
    ))
   &ScaleColor(
    Scale Alpha=&MergeHelper_FloatSwitch(
     0=1
     1=1
     2=&FloatFromCurve(FloatCurve=DataInterface() {propertyValues {Curve {
         defaultValue=1
         keys [
          {arriveTangent=1.3708, interpMode=RCIM_Cubic, leaveTangent=1.3708, tangentMode=RCTM_User, time=0, value=0}
          {arriveTangentWeight=0.0469, interpMode=RCIM_Cubic, leaveTangentWeight=0.2197, tangentMode=RCTM_User, time=0.1, value=1}
          {arriveTangentWeight=0.3342, interpMode=RCIM_Cubic, leaveTangentWeight=0.1014, tangentMode=RCTM_User, time=0.6047, value=1}
          {arriveTangent=-1.7747, interpMode=RCIM_Cubic, leaveTangent=-1.7747, tangentMode=RCTM_User, time=1, value=0}
         ]
        }}})
     3=?
     Index=&SpawnGroup()
     Num=::NewEnumerator1
    )
    Scale RGB=&MergeHelper_VectorSwitch(
     0=&Multiply_VectorByFloat(
      Vector=&MakeVectorFromLinearColorRGB(Color=&ColorFromCurve(
        ColorCurve=DataInterface() {propertyValues {
          AlphaCurve {keys [<at 0 1 RCIM_Cubic>,<at 1 0 RCIM_Cubic>]}
          BlueCurve {keys [<at 0 1 RCIM_Cubic>,<at 1 1 RCIM_Cubic>]}
          GreenCurve {keys [<at 0 1 RCIM_Cubic>,<at 1 1 RCIM_Cubic>]}
          RedCurve {keys [<at 0 1 RCIM_Cubic>,<at 1 1 RCIM_Cubic>]}
         }}
        CurveIndex=&Frac_Float(Float=&Add_Float(A=&RandomRangeFloat(Minimum=0, Maximum=1), B=$Particles.NormalizedAge))
       ))
      Float=&FloatFromCurve(
       FloatCurve=DataInterface() {propertyValues {Curve {
          defaultValue=1
          keys [
           {interpMode=RCIM_Cubic, tangentMode=RCTM_User, time=0, value=1}
           {arriveTangent=-2.2239, interpMode=RCIM_Cubic, leaveTangent=-2.2239, tangentMode=RCTM_User, time=1, value=0.2}
          ]
         }}}
       Scale Curve=&Multiply_Float(
        A=4
        B=&FN_Fit(
         Clamp:NiagaraBool=false
         Value:NiagaraFloat=$Particles.FlickerWaveform
         SourceMin:NiagaraFloat=0.2
         SourceMax:NiagaraFloat=1.2
         DestMin:NiagaraFloat=0.75
         DestMax:NiagaraFloat=1.2
        )
       )
      )
     )
     1=&Multiply_VectorByFloat(
      Vector=&MakeVectorFromLinearColorRGB(Color=&ColorFromCurve(
        ColorCurve=DataInterface() {propertyValues {
          AlphaCurve {keys [<at 0 1 RCIM_Cubic>,<at 1 0 RCIM_Cubic>]}
          BlueCurve {keys [<at 0 1 RCIM_Cubic>,<at 1 1 RCIM_Cubic>]}
          GreenCurve {keys [<at 0 1 RCIM_Cubic>,<at 1 1 RCIM_Cubic>]}
          RedCurve {keys [<at 0 1 RCIM_Cubic>,<at 1 1 RCIM_Cubic>]}
         }}
        CurveIndex=&Frac_Float(Float=&Add_Float(A=&RandomRangeFloat(Minimum=0, Maximum=1), B=$Particles.NormalizedAge))
       ))
      Float=&FloatFromCurve(
       FloatCurve=DataInterface() {propertyValues {Curve {
          defaultValue=1
          keys [
           {interpMode=RCIM_Cubic, tangentMode=RCTM_User, time=0, value=1}
           {arriveTangent=-2.2239, interpMode=RCIM_Cubic, leaveTangent=-2.2239, tangentMode=RCTM_User, time=1, value=0.2}
          ]
         }}}
       Scale Curve=&Multiply_Float(
        A=4
        B=&FN_Fit(
         Clamp:NiagaraBool=false
         Value:NiagaraFloat=$Particles.FlickerWaveform
         SourceMin:NiagaraFloat=0.2
         SourceMax:NiagaraFloat=1.2
         DestMin:NiagaraFloat=0.75
         DestMax:NiagaraFloat=1.2
        )
       )
      )
     )
     2=&MakeVectorFromLinearColorRGB(Color=&ColorFromCurve(ColorCurve=DataInterface() {propertyValues {
         AlphaCurve {keys [<at 0 1 RCIM_Cubic>]}
         BlueCurve {keys [<at 0 2.125>, {arriveTangent=-1.125, interpMode=RCIM_Cubic, leaveTangent=-1.125, time=0.1778, value=1.0156}, <at 1 1 RCIM_Cubic>]}
         GreenCurve {keys [<at 0 2.7618>, {arriveTangent=-1.7618, interpMode=RCIM_Cubic, leaveTangent=-1.7618, time=0.1778, value=1.7965}, <at 1 1 RCIM_Cubic>]}
         RedCurve {keys [<at 0 4>, {arriveTangent=-3, interpMode=RCIM_Cubic, leaveTangent=-3, time=0.1778, value=3}, <at 1 1 RCIM_Cubic>]}
        }}))
     3=&MakeVectorFromLinearColorRGB(Color=&ColorFromCurve(ColorCurve=DataInterface() {propertyValues {
         AlphaCurve {keys [<at 0 1 RCIM_Cubic>]}
         BlueCurve {keys [<at 0 2.125>, {arriveTangent=-1.125, interpMode=RCIM_Cubic, leaveTangent=-1.125, time=0.1778, value=1.0156}, <at 1 1 RCIM_Cubic>]}
         GreenCurve {keys [<at 0 2.7618>, {arriveTangent=-1.7618, interpMode=RCIM_Cubic, leaveTangent=-1.7618, time=0.1778, value=1.7965}, <at 1 1 RCIM_Cubic>]}
         RedCurve {keys [<at 0 4>, {arriveTangent=-3, interpMode=RCIM_Cubic, leaveTangent=-3, time=0.1778, value=3}, <at 1 1 RCIM_Cubic>]}
        }}))
     Index=&SpawnGroup()
     Num=::NewEnumerator1
    )
    ScaleA=true
    ScaleRGB=true
   )
   /SetVariables(Particles.Glow03_Size=&MergeHelper_Vector2Switch(
     0=&Multiply_Vector2D_ByFloat(
      Float=&FloatFromCurve(
       FloatCurve=DataInterface() {propertyValues {Curve {keys [
           <at 0 0 RCIM_Cubic>
           <at 0.2 1.3 RCIM_Cubic>
           {arriveTangent=-0.1423, interpMode=RCIM_Cubic, leaveTangent=-0.1423, tangentMode=RCTM_User, time=1, value=0}
          ]}}}
       Scale Curve=&Multiply_Float(A=&MakeFloatFromVector2d(Vector 2D=$Particles.SpriteSize), B=3.5)
      )
      Vector2D=<1 1.5>
     )
     1=&Multiply_Vector2D_ByFloat(
      Float=&FloatFromCurve(
       FloatCurve=DataInterface() {propertyValues {Curve {keys [
           <at 0 0 RCIM_Cubic>
           <at 0.2 1.3 RCIM_Cubic>
           {arriveTangent=-0.1423, interpMode=RCIM_Cubic, leaveTangent=-0.1423, tangentMode=RCTM_User, time=1, value=0}
          ]}}}
       Scale Curve=&Multiply_Float(A=&MakeFloatFromVector2d(Vector 2D=$Particles.SpriteSize), B=2)
      )
      Vector2D=<1 1>
     )
     2=?
     3=?
     4=?
     Index=&SpawnGroup()
     Num=::NewEnumerator0
    )) {isSetParameters=true}
   /SetVariables(Particles.Glow03_Color=&MergeHelper_ColorSwitch(
     0=&MakeLinearColorFromVectorAndFloat(
      "Vector (RGB)"=&MakeVectorFromLinearColorRGB(Color=&MultiplyLinearColor(Linear Color=$Particles.Color, Scale Factor=&MakeVector4_FromLinearColor(LinearColor=$Particles.Color)))
      "Float (Alpha)"=&Multiply_Float(A=0.03, B=1)
     )
     1=&MakeLinearColorFromVectorAndFloat(
      "Vector (RGB)"=&MakeVectorFromLinearColorRGB(Color=&MultiplyLinearColor(Linear Color=$Particles.Color, Scale Factor=&MakeVector4_FromLinearColor(LinearColor=$Particles.Color)))
      "Float (Alpha)"=&Multiply_Float(A=0.03, B=0.5)
     )
     2=?
     3=?
     Index=&SpawnGroup()
     Num=::NewEnumerator0
    )) {isSetParameters=true}
   /SetVariables("Particles.MeshDyn 0 ":Vector4f=&MakeVector4(
     X=&FloatFromCurve(
      FloatCurve=DataInterface() {propertyValues {Curve {keys [
          <at 0.3 0.2 RCIM_Cubic>
          {arriveTangent=0.5424, interpMode=RCIM_Cubic, leaveTangent=0.5424, tangentMode=RCTM_User, time=0.605, value=0.6745}
          <at 1.0013 0.7537 RCIM_Cubic>
         ]}}}
      Scale Curve=&Lerp_Float(A=0.7, B=0.8, Alpha=&RandomRangeFloat(Minimum=0, Maximum=1))
     )
     Y=&RandomRangeFloat(Minimum=0, Maximum=1)
     Z=&FloatFromCurve(FloatCurve=DataInterface() {propertyValues {Curve {keys [
          <at 0.6 1 RCIM_Cubic>
          {arriveTangent=-0.6876, arriveTangentWeight=0.2346, interpMode=RCIM_Cubic, leaveTangent=-0.6876, tangentMode=RCTM_User, time=1, value=0.4}
         ]}}})
     W=&FloatFromCurve(FloatCurve=DataInterface() {propertyValues {Curve {
         defaultValue=1
         keys [{interpMode=RCIM_Cubic, tangentMode=RCTM_User, time=0.6, value=0.01}, {interpMode=RCIM_Cubic, tangentMode=RCTM_User, time=1, value=0.05}]
        }}})
    )) {isSetParameters=true}
   &Drag(Drag=&RandomRangeFloat(Minimum=4, Maximum=5), Use Linear Drag=true, Use Rotational Drag=true)
   &SolveForcesAndVelocity(Clamp Velocity=false, Limit Acceleration=false)
  ]
  rendererClasses [&NiagaraSpriteRendererProperties, &NiagaraMeshRendererProperties]
  renderers [
   {rendererClass=&NiagaraSpriteRendererProperties}
   {rendererClass=&NiagaraSpriteRendererProperties, rendererIndex=1}
   {rendererClass=&NiagaraSpriteRendererProperties, rendererIndex=2}
   {rendererClass=&NiagaraSpriteRendererProperties, rendererIndex=3}
   {rendererClass=&NiagaraMeshRendererProperties, rendererIndex=4}
  ]
 }
}
summary {
 systemName=NS_Aura_FilmSparkle_Start
 emitters [
  {emitterName=Flare, rendererClasses [&NiagaraSpriteRendererProperties]}
  {emitterName=Corona_Glow_Combined, rendererClasses [&NiagaraSpriteRendererProperties]}
  {
   emitterName=Motes_Flares001
   rendererClasses [&NiagaraSpriteRendererProperties, &NiagaraMeshRendererProperties]
  }
 ]
 userVariables [
  User.bIsFrontEnd {type=NiagaraBool}
  User.bIsFrontEndPreview {type=NiagaraBool}
  User.IsFrontEnd {type=NiagaraBool}
  User.IsLargeBody {type=NiagaraBool}
  User.SkeletalMesh {
   type=NiagaraDataInterfaceSkeletalMesh
   defaultValue {
    struct=NiagaraExt_VariableValue_DataInterface
    value {dataInterface=&self:NiagaraDataInterfaceSkeletalMesh_1}
   }
  }
 ]
}
